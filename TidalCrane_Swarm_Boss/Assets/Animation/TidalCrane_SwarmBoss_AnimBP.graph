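# Animation Blueprint for the TidalCrane swarm boss creature, layering look-at aiming over an idle loop.
# The idle sequence feeds DefaultSlot, so montages can override it before the aim offset applies on top.
# LookAt yaw and pitch drive the aim offset and AOAlpha sets its weight; roll is unused.
# The event graph and variable list are empty, so nothing here computes those values.
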AnimBlueprint /TidalCrane_Swarm_Boss/Assets/Animation/TidalCrane_SwarmBoss_AnimBP
target_skeleton=/TidalCrane_Swarm_Boss/Assets/Meshes/TidalCrane_Boss_Creature_Skeleton
graphs [
 AnimGraph {nodes [
   n1 = Sequence Player 'TidalCrane_SwarmBoss_IdleLoop'() {sequence=&/TidalCrane_SwarmBoss_IdleLoop}
   n2 = Get!() -> (LookAt_Roll:float, LookAt_Pitch:float, LookAt_Yaw:float)
   Output Pose(Result=AimOffset Player 'TidalCrane_SwarmBoss_AO'!(
     X:float=@n2.LookAt_Yaw
     Y:float=@n2.LookAt_Pitch
     BasePose:Pose Link Structure=Slot 'DefaultSlot'(Source=@n1)
     Alpha:float=Get AOAlpha() -> (AOAlpha:float)
    ) -> (Pose:Pose Link Structure) {blendSpace=&/TidalCrane_SwarmBoss_AO, alphaBoolBlend {blendOption=Linear}})
  ]}
 EventGraph {links [], nodes []}
]
variables []
