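# Translucent lit material for Niagara mesh particles, apparently the Queen boss's projectile, drawing a scrolling energy trail with a bright tip.
# Noise-distorted UVs feed animated sine bands of dark energy and panned smoke textures, with per-particle random speed, tightness and distortion.
# A gradient between dark and bright colors drives glow, while particle alpha and tail gradient erode the mask behind opacity and normal-direction vertex offset.
# The DarkEnergyOnly switch keeps only dark bands, without glow.
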Material /TidalCrane_Swarm_Boss/Assets/FX/M_Projectile_Queen
graph {
 properties {bAutomaticallySetUsageInEditor=false, bUsedWithNiagaraMeshParticles=true, blendMode=BLEND_Translucent, usedShadingModels=Default Lit}
 nodes [
  n1 = ParticleColor()
  n2 = Constant()
  n3 = Panner(Coordinate=TextureCoordinate() {uTiling=2}) {speedX=0.1, speedY=1.25}
  n4 = TextureObject() {samplerType=SAMPLERTYPE_LinearGrayscale, texture=/Game/Effects/Fort_Effects/Emotes/UltraLight/Texture/T_DirectionalNoise_Greyscale}
  n5 = Power(Base=MaterialFunctionCall() {materialFunction=&LinearGradient}.VGradient) {constExponent=10}
  n6 = Multiply(A=@n5, B=ScalarParameter() {defaultValue=5, parameterName=TipBrightness})
  n7 = ParticleRandom()
  n8 = TextureSample(
   UVs=Panner(Coordinate=TextureCoordinate() {uTiling=3, vTiling=3}) {speedX=0.5, speedY=1}
   Tex=TextureObject() {texture=/Game/Effects/Fort_Effects/Emotes/UltraLight/Texture/T_SoftNoise_BC7}
  ) {texture=&T_SoftNoise_BC7}
  n9 = ParticleRandom()
  n10 = ComponentMask(None=Multiply(A=ConstantBiasScale(None=@n8.RGB) {bias=-0.5, scale=2}, B=LinearInterpolate(Alpha=@n9) {constA=0.1, constB=0.15}) {constB=0.1}) {g=true, r=true}
  n11 = TextureCoordinate() {uTiling=3, vTiling=3}
  n12 = MaterialFunctionCall(UV Channel=Multiply(A=Add(A=@n10, B=@n11), B=Constant2Vector() {g=1, r=1})) {materialFunction=&LinearGradient}
  n13 = Multiply(A=@n12.UGradient) {constB=0.5}
  n14 = Frac(None=Time())
  n15 = LinearInterpolate(Alpha=ParticleRandom()) {constA=4.75, constB=6, desc=Speed}
  n16 = Subtract(
   A=Add(A=@n12.VGradient, B=Multiply(A=@n14, B=@n15) {constB=8})
   B=Multiply(A=LinearInterpolate(Alpha=@n7) {constA=0.55, constB=0.75, desc=DiagonalDistortion}, B=Sine(None=@n13))
  )
  n17 = Power(Base=Sine(None=Multiply(A=@n16) {constB=0.5}), Exp=LinearInterpolate(Alpha=ParticleRandom()) {constA=1.25, constB=3, desc=Tightness}) {constExponent=50}
  n18 = Subtract(A=Add(A=TextureSample(UVs=@n3, Tex=@n4) {texture=/Game/Effects/Fort_Effects/Textures/Tile/T_Inky_Smoke_Tile}.RGB, B=@n6), B=@n17)
  n19 = Power(Base=MaterialFunctionCall() {materialFunction=&LinearGradient}.VGradient) {constExponent=5}
  n20 = VectorParameter() {parameterName=DarkColor, defaultValue=<rgba 0.0116 0 0.0677 1>}
  n21 = LinearInterpolate(
   A=@n20.RGB
   B=VectorParameter() {parameterName=BrightColor, defaultValue=<rgba 0.4688 0 0.2407 1>}.RGB
   Alpha=Power(Base=MaterialFunctionCall() {materialFunction=&LinearGradient}.VGradient) {constExponent=2.5}
  )
  n22 = Multiply(A=Multiply(A=Multiply(A=@n18, B=@n19), B=@n21), B=ScalarParameter() {defaultValue=50, parameterName=Brightness}) {constB=50}
  MaterialFunctionCall(Emissive=Multiply(A=@n1.RGB, B=StaticSwitchParameter(True=@n2, False=@n22) {parameterName=DarkEnergyOnly})) {materialFunction=&MF_FX_EmissiveAdjustments}
  Multiply(A=@n1.RGB, B=StaticSwitchParameter(True=@n20.RGB, False=@n21) {parameterName=DarkEnergyOnly})
  n23 = Saturate(None=Subtract(A=@n17, B=@n5))
  n24 = LinearInterpolate(Alpha=ParticleRandom()) {constA=0.25, constB=0.55, desc=DiagonalDistortion}
  n25 = MaterialFunctionCall() {materialFunction=&LinearGradient}
  n26 = Sine(None=Multiply(A=@n25.UGradient) {constB=0.5})
  n27 = Constant() {r=1}
  n28 = LinearInterpolate(Alpha=ParticleRandom()) {constA=1, constB=1.5}
  n29 = Add(A=Multiply(A=TextureCoordinate() {vTiling=2}, B=AppendVector(A=@n27, B=@n28)), B=@n10)
  n30 = LinearInterpolate(A=OneMinus(None=@n25.UGradient), B=@n25.UGradient, Alpha=Round(None=ParticleRandom())) {constAlpha=0}
  n31 = Panner(
   Coordinate=Add(A=Multiply(A=@n24, B=@n26) {constB=0.5}, B=Add(A=@n29, B=@n30))
   Speed=LinearInterpolate(Alpha=ParticleRandom()) {constA=3.25, constB=5, desc=ScrollSpeed}
  ) {speedX=1, speedY=2}
  n32 = Add(
   A=TextureSample(
    UVs=@n31
    Tex=TextureObject() {samplerType=SAMPLERTYPE_Grayscale, texture=/Game/Effects/Fort_Effects/Textures/Smoke/T_Smoke_Trail_Soft_NoClamp}
   ) {samplerType=SAMPLERTYPE_LinearGrayscale, texture=/Game/Effects/Fort_Effects/Textures/Tile/T_CellNoise_Tile_M}.R
   B=TextureSample(
    UVs=@n31
    Tex=TextureObject() {samplerType=SAMPLERTYPE_LinearColor, texture=/Game/Effects/Fort_Effects/Textures/T_Contrail_StinkyCloud_Ribbon}
   ) {samplerType=SAMPLERTYPE_LinearGrayscale, texture=&T_CellNoise_Tile_M}.R
  )
  n33 = StaticSwitchParameter(True=@n23, False=Add(A=Add(A=@n32, B=@n6), B=@n23) {desc="Adds Dark Energy\r\nTo Fill Spaces"}) {parameterName=DarkEnergyOnly}
  n34 = OneMinus(None=MaterialFunctionCall(UV Channel=@n29) {materialFunction=&LinearGradient}.VGradient)
  n35 = Subtract(A=Subtract(A=Saturate(None=@n33), B=OneMinus(None=@n1.A)), B=Multiply(A=Power(Base=@n34) {constExponent=3}) {constB=2})
  Multiply(A=Multiply(A=@n35, B=VertexNormalWS()), B=ScalarParameter() {defaultValue=10, parameterName=WPOffset}) {constB=10}
  Saturate(None=@n35)
 ]
}
