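# Unlit translucent slash effect material, apparently for the Queen boss, usable on Niagara mesh particles.
# A panning cylinder-gradient streak plus two scrolling noise samples, tinted by particle color and scaled by time-of-day HDR multipliers, drives emissive.
# Opacity reuses that pattern with particle alpha and a V-gradient mask, then applies near-camera and depth fades.
# Static switches add optional power contrast, vertex-color masking and view-angle edge fading.
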Material /TidalCrane_Swarm_Boss/Assets/FX/M_Queen_Slash
graph {
 properties {
  bUsedWithNiagaraMeshParticles=true
  blendMode=BLEND_Translucent
  pixelDepthOffsetMode=PDOM_Legacy
  shadingModel=MSM_Unlit
  usedShadingModels=Unlit
 }
 nodes [
  n1 = ScalarParameter() {defaultValue=250, parameterName=Near Camera Fade}
  n2 = ParticleColor()
  n3 = Multiply(A=Time(), B=ScalarParameter() {defaultValue=1, parameterName=TimeScale})
  n4 = TextureSample(UVs=Panner(Coordinate=TextureCoordinate(), Time=@n3) {speedY=-2}) {samplerSource=SSM_Wrap_WorldGroupSettings, texture=/Game/Global/Textures/Default/Masks_and_Gradients/CylGradient}
  n5 = VectorParameter() {parameterName=UVs, defaultValue=<rgba 5 1 10 0.25>}
  n6 = Panner(Coordinate=Multiply(A=TextureCoordinate(), B=AppendVector(A=@n5.R, B=@n5.R)), Time=@n3) {speedY=-2}
  n7 = Panner(Coordinate=Multiply(A=TextureCoordinate(), B=AppendVector(A=@n5.B, B=@n5.B)), Time=@n3) {speedY=-2}
  n8 = Add(
   A=Multiply(A=Power(Base=@n4.G) {constExponent=4}) {constB=60}
   B=Add(
    A=TextureSampleParameter2D(UVs=@n6) {parameterName=Tex1, texture=&T_NoisePack_01}.R
    B=TextureSampleParameter2D(UVs=@n7) {parameterName=Tex2, texture=&T_NoisePack_01}.G
   )
  )
  n9 = MaterialFunctionCall() {materialFunction=&LinearGradient}
  n10 = Multiply(
   A=@n9.VGradient
   B=Power(Base=OneMinus(None=@n9.VGradient), Exp=ScalarParameter() {defaultValue=3, parameterName=GradientPower}) {constExponent=3}
  )
  n11 = Multiply(A=@n2.A, B=Multiply(A=@n8, B=Multiply(A=@n10) {constB=2}))
  n12 = ScalarParameter() {defaultValue=2, parameterName=Power}
  n13 = Multiply(A=@n11, B=StaticSwitchParameter(True=Power(Base=@n11, Exp=@n12), False=Constant() {r=1}) {parameterName=Use_Power})
  n14 = StaticSwitchParameter(True=Multiply(A=@n13, B=VertexColor().R), False=@n13) {parameterName=Use Vert Color}
  n15 = DotProduct(A=VertexNormalWS(), B=CameraVectorWS())
  n16 = Multiply(A=@n14, B=Power(Base=Abs(None=@n15), Exp=ScalarParameter() {defaultValue=2, parameterName=Fresnel Power}))
  DepthFade(
   Opacity=Multiply(
    A=MaterialFunctionCall(Distance=@n1) {materialFunction=&Near_Camera_Fade}.Result
    B=StaticSwitchParameter(True=@n16, False=@n14) {parameterName=Use Fresnel}
   )
   FadeDistance=ScalarParameter() {defaultValue=25, parameterName=Depth}
  )
  n17 = ScalarParameter() {defaultValue=0.5, parameterName=HDRMult_3}
  n18 = ScalarParameter() {defaultValue=1, parameterName=HDRMult_2}
  n19 = ScalarParameter() {defaultValue=0.5, parameterName=HDRMult_1}
  n20 = ScalarParameter() {defaultValue=0.35, parameterName=HDRMult}
  n21 = Multiply(A=@n2.RGB, B=@n8)
  n22 = ScalarParameter() {defaultValue=1, parameterName=Emissive Scale}
  MaterialFunctionCall(Emissive=Multiply(
    A=MaterialFunctionCall(Color_Morning=@n17, Color_Day=@n18, Color_Evening=@n19, Color_Night=@n20) {materialFunction=&Color_Time_of_Day}.Result
    B=Multiply(A=@n21, B=@n22)
   )) {materialFunction=&MF_FX_EmissiveAdjustments}
 ]
}
