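# Niagara effect for the Queen's projectile: a purple-violet energy ball with a dark, burning trail.
# A sine-pulsed plasma sprite with a light forms the core, wrapped in rotating energy-wave meshes split across two renderers by VisibilityTag.
# Airflow and dark spline meshes re-pin to the owner each update and orient by its velocity.
# Curl-noise debris, spittle drops, time-of-day-dimmed flipbook smoke and gravity-driven peeling meshes with glowing edge colour add the shedding detail.
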NiagaraSystem /TidalCrane_Swarm_Boss/Assets/FX/NS_Projectile_Queen
dependencies {
 usedDataInterfaces [
  &NiagaraDataInterfaceArrayFloat3
  &NiagaraDataInterfaceArrayInt32
  &NiagaraDataInterfaceCurve
  &NiagaraDataInterfaceColorCurve
  &NiagaraDataInterfaceVectorCurve
  &NiagaraDataInterfaceSpriteRendererInfo
  &NiagaraDataInterfaceMeshRendererInfo
 ]
 usedDynamicInputs [&Lerp_Vector]
 usedModules [&SystemState]
}
emitters {
 Convergent_airflow_start001 {
  emitterUpdateScript [
   &EmitterState(Inactive Response=::NewEnumerator1, Life Cycle Mode=::NewEnumerator1, Loop Delay=0.3, UseLoopDelay=false)
   &SpawnBurst_Instantaneous(Spawn Count=3, Use Spawn Probability=false, UseLoopCountLimit=false)
   &SpawnRate(SpawnRate=6, Use Spawn Probability=false) {enabled=false}
  ]
  particleSpawnScript [
   &InitializeParticle(
    Color=<rgba 1 1 1 0.5>
    Lifetime=$Emitter.CurrentLoopDuration
    Mesh Scale Max=<6 3 3>
    Mesh Scale Min=<4 1.75 1.75>
    Mesh Scale Mode=::NewEnumerator2
    Position Offset=<-50 0 0>
    Position Offset Coordinate Space=::Local
    UsePositionOffset=true
    Write Material Random=true
   ) {module=InitializeParticle001}
   &InitialMeshOrientation(
    Mesh Orientation Mode=::NewEnumerator2
    Orientation Coordinate Space=::Local
    Orientation Vector=<-1 0 0>
    Rotation=&RandomRangeVector(Minimum=<0 0 0>, Maximum=<1 0 0>)
    Use Rotation Vector=true
   ) {module=InitialMeshOrientation001}
  ]
  particleUpdateScript [
   &ParticleState()
   &ScaleColor(
    Scale Alpha=&Multiply_Float(
     A=&FloatFromCurve(FloatCurve=DataInterface() {propertyValues {Curve {keys [
          {interpMode=RCIM_Cubic, tangentMode=RCTM_User, time=0, value=0}
          {interpMode=RCIM_Cubic, tangentMode=RCTM_User, time=0.5, value=1}
          {tangentMode=RCTM_User, time=1, value=0}
         ]}}})
     B=0.4
    )
    Scale RGB=&Multiply_VectorByFloat(
     Vector=&MakeVectorFromLinearColorRGB(Color=&ColorFromCurve(ColorCurve=DataInterface() {propertyValues {
         AlphaCurve {keys []}
         BlueCurve {keys [<at 0.0337 0.5>,<at 0.9172 0.5 RCIM_Cubic>]}
         GreenCurve {keys [<at 0.0337 0.33>,<at 0.9172 0.15 RCIM_Cubic>]}
         RedCurve {keys [<at 0.0337 0.2573>,<at 0.9172 0.325 RCIM_Cubic>]}
        }}))
     Float=&FloatFromCurve(
      FloatCurve=DataInterface() {propertyValues {Curve {keys [<at 0 1 RCIM_Cubic>,<at 1 0.5>]}}}
      Scale Curve=&RandomRangeFloat(Minimum=1.5, Maximum=2)
     )
    )
    ScaleA=true
    ScaleRGB=true
   )
   &ScaleMeshSize(Scale Factor=&Multiply_VectorByFloat(
     Vector=&Multiply_Vector(A=<1.3 1 0.5>, B=<1.3 2 1>)
     Float=&Multiply_Float(
      A=&FloatFromCurve(FloatCurve=DataInterface() {propertyValues {Curve {keys [
           {interpMode=RCIM_Cubic, tangentMode=RCTM_User, time=0, value=0.5}
           {arriveTangent=-0.752, interpMode=RCIM_Cubic, leaveTangent=-0.752, tangentMode=RCTM_User, time=1, value=0.3}
          ]}}})
      B=2
     )
    ))
   &DynamicMaterialParameters(
    Index 0 Param 3=&FloatFromCurve(FloatCurve=DataInterface() {propertyValues {Curve {keys [{tangentMode=RCTM_User, time=0, value=0}, {tangentMode=RCTM_User, time=1, value=1}]}}})
    Index 1 Param 1=&RandomRangeFloat(Minimum=7, Maximum=10)
    Index 1 Param 2=10
    Index 2 Param 1=&RandomRangeFloat(Minimum=100, Maximum=75)
    Index 2 Param 2=1
    Module.Index0Param0Write=false
    Module.Index0Param1Write=false
    Module.Index0Param2Write=true
    Module.Index0Param3Write=false
    Module.Index1Param0Write=true
    Module.Index1Param1Write=true
    Module.Index1Param2Write=true
    Module.Index1Param3Write=true
    Module.Index2Param0Write=true
    Module.Index2Param1Write=true
    Module.Index2Param2Write=true
    Module.Index2Param3Write=true
    Module.Index3Param0Write=true
    Module.Index3Param1Write=true
    Module.Index3Param2Write=true
    Module.Index3Param3Write=true
    Write Parameter Index 1=true
    Write Parameter Index 2=true
    Write Parameter Index 3=true
   ) {module=DynamicMaterialParameters001}
   &UpdateMeshOrientation(
    Bank Rate=0
    Banking Angle Min / Max:Vector2f=<0 10>
    "Foward Vector (Mesh Space)":Vector3f=<-1 0 0>
    Initial Forward Vector:Vector3f=$Engine.Owner.SystemXAxis
    Orientation Method=::NewEnumerator3
    Particle Velocity:Vector3f=$Engine.Owner.Velocity
    Velocity Orientation Convergence Rate=2.5
   )
   /SetVariables(Particles.Position=&AddVectorToPosition(Position=$Engine.Owner.Position, Vector=&TransformVector(Vector=<50 0 50>, Destination Space=::World))) {isSetParameters=true}
  ]
  rendererClasses [&NiagaraMeshRendererProperties]
  renderers [{rendererClass=&NiagaraMeshRendererProperties, rendererIndex=0}]
 }
 DarkEnergyWaves {
  emitterSpawnScript [/SetVariables(Emitter.RandomisedColour:LinearColor=&RandomRangeLinearColor(Minimum=<rgba 0.1875 0.041 0.1846 1>, Maximum=<rgba 0.4471 0.518 1 1>)) {isSetParameters=true}]
  emitterUpdateScript [
   &EmitterState(Inactive Response=::NewEnumerator1, Life Cycle Mode=::NewEnumerator1, UseLoopDelay=false)
   &SpawnBurst_Instantaneous(Spawn Count=3, Use Spawn Probability=false, UseLoopCountLimit=false)
  ]
  particleSpawnScript [
   &InitializeParticle(
    Color=$Emitter.RandomisedColour
    Lifetime=$Emitter.CurrentLoopDuration
    Mesh Scale=&Multiply_Vector(
     A=&SelectVectorFromArray(
      Direct Array Index=$Engine.ExecIndex
      Vector Selection Array=DataInterface() {propertyValues {FloatData [<12 5.5 5.5>, <8 4 4>, <12 5.5 5.5>]}}
     )
     B=&RandomRangeVector(Minimum=<0.65 1 1>, Maximum=<1 1 1>)
    )
    Mesh Scale Mode=::NewEnumerator4
    Position Offset=<100 0 0>
    UsePositionOffset=true
    Write Material Random=true
   )
   /SetVariables(
    Particles.DarkEnergyScale:Vector3f=&Multiply_Vector(A=$Particles.Scale, B=&RandomRangeVector(Minimum=<0.75 1.15 1.15>, Maximum=<1 1.25 1.25>))
    Particles.VisibilityTag=&SelectIntFromArray(Direct Array Index=$Engine.ExecIndex, Int Selection Array=DataInterface() {propertyValues {IntData [0,0,1]}})
   ) {isSetParameters=true}
  ]
  particleUpdateScript [
   &ParticleState()
   &UpdateMeshOrientation(Rotation Rate=&RandomRangeFloat(Minimum=0.3, Maximum=0.75))
  ]
  rendererClasses [&NiagaraMeshRendererProperties]
  renderers [
   {rendererClass=&NiagaraMeshRendererProperties, rendererIndex=0}
   {rendererClass=&NiagaraMeshRendererProperties, rendererIndex=1}
  ]
 }
 DarkSpline {
  emitterUpdateScript [
   &EmitterState()
   &SpawnBurst_Instantaneous(Spawn Count=2, Use Spawn Probability=false, UseLoopCountLimit=false)
   &SpawnRate(SpawnRate=10, Use Spawn Probability=false)
  ]
  particleSpawnScript [
   &InitializeParticle(
    Color Maximum=<rgba 0.009 0.005 0.0312 1>
    Color Minimum=<rgba 0.0998 0.0295 0.125 1>
    Color Mode=::NewEnumerator2
    Lifetime Max=0.75
    Lifetime Min=0.45
    Lifetime Mode=::NewEnumerator1
    Mass Mode=::NewEnumerator3
    Mesh Scale Max=<4 4 8>
    Mesh Scale Min=<1.5 1.5 4>
    Mesh Scale Mode=::NewEnumerator2
    Sprite Rotation Mode=::NewEnumerator2
    Sprite Size Mode=::NewEnumerator1
    Uniform Sprite Size Max=12
    Uniform Sprite Size Min=6
    UsePositionOffset=false
    Write Material Random=true
   )
   &InitialMeshOrientation(
    Mesh Axis To Orient=<0 0 1>
    Mesh Orientation Mode=::NewEnumerator2
    Orientation Coordinate Space=::Local
    Rotation=&RandomRangeVector(Minimum=<0 0 0>, Maximum=<0 0 1>)
    Use Rotation Vector=true
   )
   /SetVariables(
    Particles.InitialOffset:Vector3f=&Subtract_Position(A=$Particles.Position, B=$Engine.Owner.Position)
    Particles.InitialOrientationVector:Vector3f=&TransformVector(Vector=$Engine.Owner.SystemXAxis, Destination Space=::World)
   ) {isSetParameters=true}
  ]
  particleUpdateScript [
   &ParticleState()
   &UpdateMeshOrientation(Rotation Rate=&RandomRangeFloat(Minimum=-0.5, Maximum=0.5), Rotation Vector=<0 0 1>) {module=UpdateMeshOrientation001}
   &ScaleMeshSize(Scale Factor=&VectorFromCurve(
     VectorCurve=DataInterface() {propertyValues {
       XCurve {keys [
         {arriveTangent=1.9584, interpMode=RCIM_Cubic, leaveTangent=1.9584, tangentMode=RCTM_User, time=0, value=0.2}
         {arriveTangent=1.1179, interpMode=RCIM_Cubic, leaveTangent=1.1179, tangentMode=RCTM_User, time=0.3, value=0.7}
         {arriveTangent=0.0224, interpMode=RCIM_Cubic, leaveTangent=0.0224, tangentMode=RCTM_User, time=1, value=1}
        ]}
       YCurve {keys [
         {arriveTangent=1.9584, interpMode=RCIM_Cubic, leaveTangent=1.9584, tangentMode=RCTM_User, time=0, value=0.2}
         {arriveTangent=1.1179, interpMode=RCIM_Cubic, leaveTangent=1.1179, tangentMode=RCTM_User, time=0.3, value=0.7}
         {arriveTangent=0.0224, interpMode=RCIM_Cubic, leaveTangent=0.0224, tangentMode=RCTM_User, time=1, value=1}
        ]}
       ZCurve {keys [
         {arriveTangent=1.8004, interpMode=RCIM_Cubic, leaveTangent=1.8004, tangentMode=RCTM_User, time=0, value=0.2}
         {arriveTangent=0.0224, interpMode=RCIM_Cubic, leaveTangent=0.0224, tangentMode=RCTM_User, time=1, value=1}
        ]}
      }}
     Scale Curve=<3 3 3>
    ))
   &ScaleColor(
    Scale Alpha=&FloatFromCurve(
     FloatCurve=DataInterface() {propertyValues {Curve {keys [
         {arriveTangent=2.8016, interpMode=RCIM_Cubic, leaveTangent=2.8016, tangentMode=RCTM_User, time=0, value=0.3}
         {interpMode=RCIM_Cubic, tangentMode=RCTM_User, time=0.5, value=1}
         {tangentMode=RCTM_User, time=1, value=0}
        ]}}}
     Scale Curve=10
    )
    ScaleA=true
    ScaleRGB=true
   )
   &DynamicMaterialParameters(
    Index 3 Param 1=&FloatFromCurve(
     FloatCurve=DataInterface() {propertyValues {Curve {keys [{tangentMode=RCTM_User, time=0, value=0.5}, {tangentMode=RCTM_User, time=1, value=1}]}}}
     Scale Curve=&RandomRangeFloat(Minimum=5, Maximum=12)
    )
    Index 3 Param 3=&RandomRangeFloat(Minimum=0, Maximum=1)
    Module.Index3Param0Write=true
    Module.Index3Param1Write=false
    Module.Index3Param2Write=true
    Module.Index3Param3Write=false
    Write Parameter Index 0=false
    Write Parameter Index 3=true
   )
   &UpdateMeshOrientation(
    Bank Rate=0
    Banking Angle Min / Max:Vector2f=<0 10>
    "Foward Vector (Mesh Space)":Vector3f=<0 0 -1>
    Initial Forward Vector:Vector3f=&Multiply_VectorByFloat(Vector=$Engine.Owner.SystemXAxis, Float=-1)
    Orientation Method=::NewEnumerator3
    Particle Velocity:Vector3f=$Engine.Owner.Velocity
    Velocity Orientation Convergence Rate=3
    Write to Orientation:ENiagara_AttributeSamplingApplyOutput=::NewEnumerator1
   )
   /SetVariables(Particles.Position=&AddVectorToPosition(Position=$Engine.Owner.Position, Vector=&TransformVector(Vector=<0 0 10>, Destination Space=::World))) {isSetParameters=true}
  ]
  rendererClasses [&NiagaraMeshRendererProperties]
  renderers [{rendererClass=&NiagaraMeshRendererProperties, rendererIndex=0}]
 }
 HotDebris {
  emitterUpdateScript [
   &EmitterState()
   &SpawnRate(SpawnRate=50, Use Spawn Probability=false)
  ]
  particleSpawnScript [
   &InitializeParticle(
    Color=<rgba 20 0 19.036 1>
    Lifetime=&UniformRangedFloat(Minimum=0.25, Maximum=0.5)
    Sprite Rotation Mode=::NewEnumerator2
    Sprite Size Mode=::NewEnumerator1
    Uniform Sprite Size Max=15
    UsePositionOffset=false
    Write Material Random=true
   ) {module=InitializeParticle001}
   &ShapeLocation(
    MaskSpawnGroup=false
    Non Uniform Scale=<2 1 1>
    Offset=<-75 0 0>
    Offset Coordinate Space=::Local
    Offset Mode=::NewEnumerator0
    Override Seed=false
    Sphere Radius=50
    UseOwnerScale=true
   )
   &AddVelocity(
    Module.Use Falloff Distance=false
    Origin Offset=<150 0 0>
    Override Seed=false
    Velocity Mode=::NewEnumerator1
    Velocity Speed=&RandomRangeFloat(Minimum=550, Maximum=1250)
   ) {module=AddVelocity001}
   &SubUVAnimation(
    Evaluation Type=::NewEnumerator0
    "Sprite Renderer "=DataInterface() {propertyValues {SpriteRenderer=&self:HotDebris_0.NiagaraSpriteRendererProperties_1}}
    SubUV Animation Mode:ENiagara_SubUVLookupModeV2=::NewEnumerator2
    UseEndFrame=false
    UseStartFrame=false
   )
  ]
  particleUpdateScript [
   &ParticleState()
   &ScaleSpriteSize(Scale Factor=&Vector2DFromFloat(Value=&FloatFromCurve(FloatCurve=DataInterface() {propertyValues {Curve {keys [<at 0 0.5>,<at 0.1 0.9>,<at 1 0>]}}})))
   &GravityForce()
   &Drag(Drag=&RandomRangeFloat(Minimum=3, Maximum=5), Use Linear Drag=true, Use Rotational Drag=false)
   &CurlNoiseForce(Bias Noise=false, Noise Frequency=20, Noise Strength=3500, Pan Noise=false)
   &SolveForcesAndVelocity(Clamp Velocity=false, Limit Acceleration=false)
   &ScaleColor(
    Scale RGB=&VectorFromFloat(Value=&FloatFromCurve(FloatCurve=DataInterface() {propertyValues {Curve {keys [
          {arriveTangent=-2.3795, interpMode=RCIM_Cubic, leaveTangent=-2.3795, tangentMode=RCTM_User, time=-0.0046, value=1.0113}
          <at 1.0011 0.002 RCIM_Cubic>
         ]}}}))
    ScaleA=false
    ScaleRGB=true
   )
  ]
  rendererClasses [&NiagaraSpriteRendererProperties]
  renderers [{rendererClass=&NiagaraSpriteRendererProperties}]
 }
 IriSmoke {
  emitterUpdateScript [
   &EmitterState()
   &SpawnRate(SpawnRate=10, Use Spawn Probability=false)
  ]
  particleSpawnScript [
   &InitializeParticle(
    Lifetime Max=0.55
    Lifetime Min=0.35
    Lifetime Mode=::NewEnumerator1
    Sprite Rotation Mode=::NewEnumerator2
    Sprite Size Mode=::NewEnumerator1
    Uniform Sprite Size Max=350
    Uniform Sprite Size Min=275
    UsePositionOffset=false
    Write Material Random=true
   ) {module=InitializeParticle001}
   /SetVariables(Particles.RandomFrame=&RandomRangeInt(Minimum=0, Maximum=63)) {isSetParameters=true}
   &SphereLocation(
    MaskSpawnGroup=false
    Override Seed=false
    Sphere Radius=3
    Surface Expansion Mode=::NewEnumerator1
    Surface Only=true
    UseNonUniformScale=false
    UseOffset=false
    UseSphereOrientation=false
   )
   &AddVelocityFromPoint(
    Module.Use Falloff Distance=false
    Origin Offset=<100 0 0>
    Velocity Strength=&RandomRangeFloat(Minimum=750, Maximum=1750)
   )
  ]
  particleUpdateScript [
   &Drag(Drag=&RandomRangeFloat(Minimum=0.5, Maximum=2), Use Linear Drag=true, Use Rotational Drag=false)
   &SolveForcesAndVelocity(Clamp Velocity=false, Limit Acceleration=false)
   &ParticleState()
   &ScaleColor(
    Scale Alpha=&FloatFromCurve(
     FloatCurve=DataInterface() {propertyValues {Curve {keys [<at 0 1>,<at 1 0>]}}}
     Scale Curve=&FNTimeOfDayScalar(Night Scale=1, Morning Scale=0.3, Evening Scale=0.5)
    )
    Scale RGB=&VectorFromFloat(Value=&RandomRangeFloat(Minimum=0.1, Maximum=0.35))
    ScaleA=true
    ScaleRGB=true
   )
   &ScaleColor(
    Scale Alpha=&FloatFromCurve(
     FloatCurve=DataInterface() {propertyValues {Curve {keys [
         {arriveTangent=2.5779, interpMode=RCIM_Cubic, leaveTangent=2.5779, tangentMode=RCTM_User, time=0, value=0}
         {arriveTangent=0.3312, interpMode=RCIM_Cubic, leaveTangent=-0.2993, tangentMode=RCTM_Break, time=0.3, value=1}
         {arriveTangent=-0.6165, interpMode=RCIM_Cubic, leaveTangent=-0.6165, tangentMode=RCTM_User, time=1, value=0}
        ]}}}
     Scale Curve=&RandomRangeFloat(Minimum=0.5, Maximum=1)
    )
    Scale RGB=&MakeVectorFromLinearColorRGB(Color=&ColorFromCurve(ColorCurve=DataInterface() {propertyValues {
        AlphaCurve {keys [<at 0 1 RCIM_Cubic>]}
        BlueCurve {keys [
          <at 0 1>
          <at 0.2106 1>
          {arriveTangent=-0.6161, interpMode=RCIM_Cubic, leaveTangent=-0.6161, time=0.6317, value=1}
          <at 1 0.3839 RCIM_Cubic>
         ]}
        GreenCurve {keys [
          <at 0 0.2913>
          <at 0.2106 0.0625>
          {arriveTangent=-0.2392, interpMode=RCIM_Cubic, leaveTangent=-0.2392, time=0.6317, value=0}
          <at 1 0.0521 RCIM_Cubic>
         ]}
        RedCurve {keys [
          <at 0 0.1823>
          <at 0.2106 0.1094>
          {arriveTangent=0.8177, interpMode=RCIM_Cubic, leaveTangent=0.8177, time=0.6317, value=0.1667}
          <at 1 1 RCIM_Cubic>
         ]}
       }}))
    ScaleA=true
    ScaleRGB=true
   ) {module=ScaleColor001}
   &ScaleSpriteSize(Scale Factor=&Vector2DFromFloat(Value=&FloatFromCurve(FloatCurve=DataInterface() {propertyValues {Curve {keys [
          {arriveTangent=2.5076, interpMode=RCIM_Cubic, leaveTangent=2.5076, tangentMode=RCTM_User, time=0, value=0.25}
          {arriveTangent=0.0897, interpMode=RCIM_Cubic, leaveTangent=0.0897, tangentMode=RCTM_User, time=1, value=1}
         ]}}})))
   &FN_SubUVAnimation(End Frame=&Add_Integer(A=$Particles.RandomFrame, B=63), Start Frame=$Particles.RandomFrame, SubUV Loop Count=0.75)
  ]
  rendererClasses [&NiagaraSpriteRendererProperties]
  renderers [{rendererClass=&NiagaraSpriteRendererProperties}]
 }
 PeelingBurningElement {
  emitterUpdateScript [
   &EmitterState()
   &SpawnRate(Spawn Probability=0.5, SpawnRate=75, Use Spawn Probability=true)
  ]
  particleSpawnScript [
   &InitializeParticle(
    Color=<rgba 0.019 0.009 0.035 1>
    Lifetime Max=1.25
    Lifetime Min=0.55
    Lifetime Mode=::NewEnumerator1
    Mass Mode=::NewEnumerator1
    Mesh Renderer Array Visibility Mode=::NewEnumerator3
    Mesh Renderer Info=DataInterface() {propertyValues {MeshRenderer=&self:PeelingBurningElement_0.NiagaraMeshRendererProperties_1}}
    Mesh Scale Mode=::NewEnumerator1
    Mesh Uniform Scale Max=1
    Mesh Uniform Scale Min=0.25
    Position Offset=<15 0 0>
    Position Offset Coordinate Space=::Local
    UsePositionOffset=true
    Write Material Random=true
   )
   /SetVariables(Particles.ParticleRandom:NiagaraFloat=&RandomRangeFloat(Minimum=0, Maximum=1)) {isSetParameters=true}
   &InheritVelocity(Inherited Velocity Amount Scale=<0.85 0.85 0.85>, Speed Limit=true, UseSpeedThreshold=false)
   &ShapeLocation(
    MaskSpawnGroup=false
    Offset=<50 0 0>
    Offset Coordinate Space=::Local
    Offset Mode=::NewEnumerator0
    Override Seed=false
    Ring / Disc Distribution Mode:ENiagaraSphereDistributionMode=::NewEnumerator1
    Ring Radius=10
    Rotation Mode=::NewEnumerator1
    Shape Primitive=::NewEnumerator4
    U Position=$Particles.ParticleRandom
    UseOwnerScale=true
   )
   &InitialMeshOrientation(
    Mesh Axis To Orient=<0 1 0>
    Mesh Orientation Mode=::NewEnumerator2
    Rotation=&MakeVector(
     X=0
     Y=&Add_Float(A=$Particles.ParticleRandom, B=&RandomRangeFloat(Minimum=-0.05, Maximum=0.05))
     Z=0
    )
    Use Rotation Vector=true
   )
   &AddVelocity(
    Module.Use Falloff Distance=false
    Origin Offset=<55 0 0>
    Override Seed=false
    Velocity Mode=::NewEnumerator1
    Velocity Speed=&RandomRangeFloat(Minimum=150, Maximum=450)
   )
  ]
  particleUpdateScript [
   &ParticleState()
   /SetVariables(Particles.EdgeColor:LinearColor=<rgba 5.9325 1.4712 8 1>) {isSetParameters=true}
   &GravityForce(Gravity=<0 0 -550>)
   &Drag(Drag=0.25, Use Linear Drag=true, Use Rotational Drag=true)
   &SolveForcesAndVelocity(Clamp Velocity=false, Limit Acceleration=false)
   &DynamicMaterialParameters(
    Index 0 Param 1=&RandomRangeFloat(Minimum=0.5, Maximum=1)
    Index 0 Param 2=1
    Index 0 Param 3=&Lerp_Float(
     A=0.25
     B=0.5
     Alpha=&FloatFromCurve(FloatCurve=DataInterface() {propertyValues {Curve {keys [
          {interpMode=RCIM_Cubic, leaveTangentWeight=0.7892, tangentMode=RCTM_User, time=0.2, value=0}
          {arriveTangent=-0.0377, interpMode=RCIM_Cubic, leaveTangent=-0.0377, tangentMode=RCTM_User, time=1, value=1}
         ]}}})
    )
    Index 0 Param 4=&Lerp_Float(
     A=10
     B=50
     Alpha=&FloatFromCurve(FloatCurve=DataInterface() {propertyValues {Curve {keys [{interpMode=RCIM_Cubic, tangentMode=RCTM_User, time=0, value=0.2}, {interpMode=RCIM_Cubic, tangentMode=RCTM_User, time=1, value=1}]}}})
    )
    Index 1 Linear Color:LinearColor=$Particles.EdgeColor
    Module.Index0Float0Write=true
    Module.Index0Float1Write=true
    Module.Index0Float2Write=true
    Module.Index0Float3Write=true
    Write Parameter Index 1=::NewEnumerator5
   )
   &UpdateMeshOrientation(
    Rotation Rate=&RandomRangeFloat(Minimum=-0.35, Maximum=-0.5)
    Rotation Vector=&RandomRangeVector(Minimum=<0.75 -0.35 0>, Maximum=<2 0.35 0>)
   ) {module=UpdateMeshOrientation002}
  ]
  rendererClasses [&NiagaraMeshRendererProperties]
  renderers [{rendererClass=&NiagaraMeshRendererProperties, rendererIndex=0}]
 }
 Plasma_Center_Energy {
  emitterUpdateScript [
   &EmitterState(Inactive Response=::NewEnumerator1, Life Cycle Mode=::NewEnumerator1, UseLoopDelay=false)
   &SpawnBurst_Instantaneous(Spawn Count=1, Use Spawn Probability=false)
  ]
  particleSpawnScript [&InitializeParticle(
    Color Mode=::NewEnumerator0
    Lifetime=$Emitter.CurrentLoopDuration
    Position Offset=<75 0 0>
    Sprite Size Mode=::NewEnumerator1
    Uniform Sprite Size Max=250
    Uniform Sprite Size Min=230
    UsePositionOffset=true
    Write Material Random=true
   )]
  particleUpdateScript [
   &ParticleState()
   &ScaleSpriteSize(
    Scale Sprite Size Mode=::NewEnumerator0
    Uniform Scale Factor=&Sine(Normalized Angle=$Emitter.Age, Period=0.5, Scale=0.1, Bias=1)
   )
   &Color(Color=<rgba 5.6 0 20 1>, Scale Alpha=0.5)
   &FN_Light_Attributes(
    Color=<rgba 0.2223 0 10 1>
    Create Color Attribute=true
    Create Exponent Attribute=true
    Create Radius Attribute=true
    Exponent=5
    Radius=200
   ) {module=FN_Light_Attributes001}
  ]
  rendererClasses [&NiagaraSpriteRendererProperties, &NiagaraLightRendererProperties]
  renderers [
   {rendererClass=&NiagaraSpriteRendererProperties}
   {rendererClass=&NiagaraSpriteRendererProperties, rendererIndex=1}
   {rendererClass=&NiagaraLightRendererProperties, rendererIndex=2}
  ]
 }
 SpittleDrops {
  emitterUpdateScript [
   &EmitterState(Life Cycle Mode=::NewEnumerator1, Loop Duration=2, UseLoopDelay=false)
   &SpawnRate(
    SpawnRate=&FloatFromCurve(
     FloatCurve=DataInterface() {propertyValues {Curve {keys [{arriveTangent=-3.5052, interpMode=RCIM_Cubic, leaveTangent=-3.5052, tangentMode=RCTM_User, time=0, value=1}, <at 1 0>]}}}
     CurveIndex=$Emitter.Age
     Scale Curve=200
    )
    Use Spawn Probability=false
   )
  ]
  particleSpawnScript [
   &InitializeParticle(
    Lifetime Min=0.25
    Lifetime Mode=::NewEnumerator1
    Mass Mode=::NewEnumerator3
    Sprite Rotation Mode=::NewEnumerator2
    Sprite Size Mode=::NewEnumerator3
    Uniform Sprite Size=&FloatFromCurve(
     FloatCurve=DataInterface() {propertyValues {Curve {keys [{arriveTangent=-2.3105, interpMode=RCIM_Cubic, leaveTangent=-2.3105, tangentMode=RCTM_User, time=0, value=1}, <at 0.5 0.35>]}}}
     CurveIndex=$Emitter.NormalizedLoopAge
     Scale Curve=&RandomRangeFloat(Minimum=12, Maximum=35)
    )
    UsePositionOffset=false
    Write Material Random=true
   )
   &ShapeLocation(MaskSpawnGroup=false, Override Seed=false, Sphere Radius=20, UseOwnerScale=true)
   &AddVelocity(
    Cone Angle=90
    Distribution Along Cone Axis=0.25
    Inner Cone Angle=45
    Module.UseConeFalloff=true
    Module.UseVelDistribution=true
    Override Seed=false
    Speed Falloff From Cone Axis=0.5
    Velocity Mode=::NewEnumerator2
    Velocity Speed=&RandomRangeFloat(Minimum=550, Maximum=1250)
   )
  ]
  particleUpdateScript [
   &ParticleState()
   &GravityForce()
   &Drag(Drag=0.25, Use Linear Drag=true, Use Rotational Drag=false)
   &ScaleColor(
    Scale Alpha=&FloatFromCurve(FloatCurve=DataInterface() {propertyValues {Curve {keys [<at -0.0046 1.0113 RCIM_Cubic>,<at 1.0011 0.002 RCIM_Cubic>]}}})
    ScaleA=true
    ScaleRGB=true
   )
   &CurlNoiseForce(Bias Noise=false, Noise Frequency=10, Noise Strength=350, Pan Noise=false)
   &SolveForcesAndVelocity(Clamp Velocity=false, Limit Acceleration=false)
   &ScaleSpriteSizeBySpeed(
    Max Scale Factor=&RandomRangeVector2D(Minimum=<1 2>, Maximum=<1 8>)
    Min Scale Factor=<0.5 0.35>
    Velocity Threshold=1500
   )
  ]
  rendererClasses [&NiagaraSpriteRendererProperties]
  renderers [{rendererClass=&NiagaraSpriteRendererProperties}]
 }
}
summary {
 systemName=NS_Projectile_Queen
 emitters [
  {emitterName=DarkEnergyWaves, rendererClasses [&NiagaraMeshRendererProperties]}
  {emitterName=IriSmoke, rendererClasses [&NiagaraSpriteRendererProperties]}
  {
   emitterName=Plasma_Center_Energy
   rendererClasses [&NiagaraSpriteRendererProperties, &NiagaraLightRendererProperties]
  }
  {emitterName=DarkSpline, rendererClasses [&NiagaraMeshRendererProperties]}
  {emitterName=Convergent_airflow_start001, rendererClasses [&NiagaraMeshRendererProperties]}
  {emitterName=SpittleDrops, rendererClasses [&NiagaraSpriteRendererProperties]}
  {emitterName=HotDebris, rendererClasses [&NiagaraSpriteRendererProperties]}
  {emitterName=PeelingBurningElement, rendererClasses [&NiagaraMeshRendererProperties]}
 ]
 userVariables [User.bIsFrontEndPreview {type=NiagaraBool}]
}
