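# Niagara ground telegraph for the Queen boss's projectile, built from two decal renderers on a single burst particle that lasts one emitter loop.
# The particle sets a fixed decal size and a vertically aligned orientation, while a separate outer decal size shrinks along a curve indexed by emitter age, suggesting a closing warning ring.
# User.ShowDecal drives particle visibility, and the HotDebris sprite emitter is disabled.
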NiagaraSystem /TidalCrane_Swarm_Boss/Assets/FX/NS_Projectile_Queen_Telegraph
dependencies=null
emitters {Flash_Flare001 {
  emitterUpdateScript [
   &EmitterState()
   &SpawnBurst_Instantaneous(Spawn Count=1, Use Spawn Probability=false)
  ]
  particleSpawnScript [
   &InitializeParticle(Color=<rgba 0.0504 0.0233 0.1719 0.25>, Lifetime=$Emitter.CurrentLoopDuration, UsePositionOffset=false, Write Material Random=true)
   /SetVariables(
    Particles.DecalFade=1
    Particles.DecalOrientation:Quat4f=&OrientMeshToVector(Look At Direction=<0 0 1>, Reference Axis Vector=<1 0 0>)
    Particles.DecalSize:Vector3f=<300 650 650>
   ) {isSetParameters=true}
  ]
  particleUpdateScript [
   &ParticleState()
   /SetVariables(Particles.Outer_Decak Size:Vector3f=&Lerp_Vector(
     A=<450 1100 1100>
     B=<256.5 580 580>
     Alpha=&FloatFromCurve(
      FloatCurve=DataInterface() {propertyValues {Curve {
         defaultValue=1
         keys [
          {arriveTangent=0.1472, interpMode=RCIM_Cubic, leaveTangent=0.1472, tangentMode=RCTM_User, time=0, value=0}
          {arriveTangent=1.2372, tangentMode=RCTM_User, time=1, value=1}
         ]
        }}}
      CurveIndex=$Emitter.Age
     )
    )) {isSetParameters=true}
   /SetVariables(Particles.Vis:NiagaraBool=$User.ShowDecal) {isSetParameters=true}
  ]
  rendererClasses [&NiagaraDecalRendererProperties]
  renderers [
   {rendererClass=&NiagaraDecalRendererProperties, rendererIndex=0}
   {rendererClass=&NiagaraDecalRendererProperties, rendererIndex=1}
  ]
 }}
summary {
 systemName=NS_Projectile_Queen_Telegraph
 emitters [
  {emitterName=Flash_Flare001, rendererClasses [&NiagaraDecalRendererProperties]}
  {bEnabled=false, emitterName=HotDebris, rendererClasses [&NiagaraSpriteRendererProperties]}
 ]
 userVariables [
  User.DespawnTime {type=NiagaraFloat, defaultValue=0, description=""}
  User.IsWater {type=NiagaraBool}
  User.ShowDecal {type=NiagaraBool, defaultValue=true}
  User.WaterPlaneLocation {type=NiagaraPosition, description="", defaultValue {}}
 ]
}
