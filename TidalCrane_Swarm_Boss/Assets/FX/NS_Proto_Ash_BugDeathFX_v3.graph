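# Niagara prototype effect for a bug death: a short, mostly purple burst of ash, goo and smoke.
# GPU ashes drift on curl noise and gravity, alongside glow flashes, light streaks, smoke and directional splashes.
# User.EnableGoo gates drops and several splash emitters through spawn probability, while User.FXColor selects an ash color and User.FXScale sizes smoke.
# Camera-distance remaps adjust size and speed, and time-of-day scalars tune streak brightness.
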NiagaraSystem /TidalCrane_Swarm_Boss/Assets/FX/NS_Proto_Ash_BugDeathFX_v3
dependencies=null
emitters {
 Ashes {
  simTarget=GPUComputeSim
  emitterUpdateScript [
   &EmitterState()
   &SpawnBurst_Instantaneous(Spawn Count=35, Use Spawn Probability=false, UseLoopCountLimit=false)
   &SpawnRate(
    SpawnRate=&FloatFromCurve(
     FloatCurve=DataInterface() {propertyValues {Curve {keys [{arriveTangent=-7.1429, interpMode=RCIM_Cubic, time=0.05, value=0}, <at 0.15 1 RCIM_Cubic>, <at 0.2 0 RCIM_Cubic>]}}}
     CurveIndex=$Emitter.Age
     Scale Curve=150
    )
    Use Spawn Probability=false
   )
   /SetVariables(Emitter.AshColor:LinearColor=&SelectLinearColorFromArray(
     Array Sampling Mode=::NewEnumerator1
     Direct Array Index=$User.FXColor
     Color Selection Array=DataInterface() {propertyValues {ColorData [<rgba 1 1 1 1>, <rgba 0 0.6717 1 1>, <rgba 0.6697 0.1914 5 1>, <rgba 0.9511 5 0.4197 1>]}}
     Random Seed=?
    )) {isSetParameters=true}
  ]
  particleSpawnScript [
   &InitializeParticle(
    Color=&LerpLinearColors(
     StartColor=<rgba 0.1368 0.1297 0.2552 1>
     EndColor=<rgba 0.2703 0.1898 1 1>
     LerpFactor=&FloatFromCurve(
      FloatCurve=DataInterface() {propertyValues {Curve {keys [
          {interpMode=RCIM_Cubic, tangentMode=RCTM_User, time=0, value=0}
          {arriveTangent=2.8251, interpMode=RCIM_Cubic, leaveTangent=2.8251, tangentMode=RCTM_User, time=1, value=1}
         ]}}}
      CurveIndex=&RandomRangeFloat(Minimum=0, Maximum=1)
     )
    )
    Lifetime Min=0.75
    Lifetime Mode=::NewEnumerator1
    Mesh Scale Mode=::NewEnumerator3
    Mesh Uniform Scale=100
    Sprite Rotation Mode=::NewEnumerator2
    Sprite Size Mode=::NewEnumerator3
    Uniform Sprite Size=&FloatFromCurve(
     FloatCurve=DataInterface() {propertyValues {Curve {keys [<at 0 2.5>,<at 0.13 1>,<at 1 0.3>]}}}
     CurveIndex=&RandomRangeFloat(Minimum=0, Maximum=1)
     Scale Curve=28
    )
    UsePositionOffset=false
    Write Material Random=true
   )
   /SetVariables() {isSetParameters=true}
   &ShapeLocation(MaskSpawnGroup=false, Override Seed=false, Sphere Radius=25, Sphere Surface Distribution=0.869, UseOwnerScale=true)
   &AddVelocity(
    Module.Use Falloff Distance=false
    Origin Offset=<0 -20 0>
    Override Seed=false
    Velocity Mode=::NewEnumerator1
    Velocity Speed=&Multiply_Float(
     A=&RandomRangeFloat(Minimum=850, Maximum=7500)
     B=&Lerp_Float(
      A=0.5
      B=0.85
      Alpha=&FloatFromCurve(
       FloatCurve=DataInterface() {propertyValues {Curve {keys [<at 0.1 0>,<at 0.3 1>,<at 1 1>]}}}
       CurveIndex=&FN_LOD_Distance_Remap(Start Distance=0, End Distance=5000, Minimum Output=0, Maximum Output=1, Clamp Output=true)
      )
     )
    )
   ) {module=AddVelocity002}
   &CameraOffset(Camera Offset Amount=20)
  ]
  particleUpdateScript [
   &ParticleState()
   &GravityForce(Gravity=&MakeVector(X=0, Y=0, Z=&RandomRangeFloat(Minimum=0, Maximum=-2500)))
   &CurlNoiseForce(
    Bias Noise=false
    Noise Frequency=10
    Noise Strength=&FloatFromCurve(
     FloatCurve=DataInterface() {propertyValues {Curve {keys [{interpMode=RCIM_Cubic, tangentMode=RCTM_Break, time=0.2, value=1}, {arriveTangent=3.0586, tangentMode=RCTM_User, time=1, value=2}]}}}
     CurveIndex=$Emitter.Age
     Scale Curve=&RandomRangeFloat(Minimum=100, Maximum=800)
    )
    Pan Noise=true
    Pan Noise Field=<0 0 2>
   )
   &Drag(
    Drag=&FloatFromCurve(
     FloatCurve=DataInterface() {propertyValues {Curve {keys [
         {interpMode=RCIM_Cubic, leaveTangentWeight=0.5018, tangentMode=RCTM_Break, time=0, value=1}
         {interpMode=RCIM_Cubic, tangentMode=RCTM_Break, time=0.2, value=0}
        ]}}}
     Scale Curve=25
    )
    Use Linear Drag=true
    Use Rotational Drag=true
   )
   &SolveForcesAndVelocity(Clamp Velocity=false, Limit Acceleration=false)
   &ScaleColor(
    Scale Alpha=&FloatFromCurve(FloatCurve=DataInterface() {propertyValues {Curve {keys [{interpMode=RCIM_Cubic, tangentMode=RCTM_Break, time=0.2, value=1}, <at 1 0>]}}})
    Scale RGB=&VectorFromFloat(Value=&Lerp_Float(
      A=0.8
      B=2
      Alpha=&FloatFromCurve(FloatCurve=DataInterface() {propertyValues {Curve {
          defaultValue=1
          keys [
           {interpMode=RCIM_Cubic, tangentMode=RCTM_User, time=0, value=1}
           {arriveTangent=-2.5, interpMode=RCIM_Cubic, leaveTangent=-2.5, time=0.1185, value=0.3709}
           {interpMode=RCIM_Cubic, tangentMode=RCTM_User, time=0.4, value=0}
          ]
         }}})
     ))
    ScaleA=true
    ScaleRGB=true
   )
   &ScaleColor(
    Scale Alpha=&FloatFromCurve(FloatCurve=DataInterface() {propertyValues {Curve {keys [{interpMode=RCIM_Cubic, tangentMode=RCTM_Break, time=0.2, value=1}, <at 1 0>]}}})
    Scale RGB=&VectorFromFloat(Value=&Lerp_Float(
      A=0.01
      B=0.2
      Alpha=&FloatFromCurve(FloatCurve=DataInterface() {propertyValues {Curve {
          defaultValue=1
          keys [
           {interpMode=RCIM_Cubic, tangentMode=RCTM_User, time=0, value=1}
           {arriveTangent=-2.5, interpMode=RCIM_Cubic, leaveTangent=-2.5, time=0.1185, value=0.3709}
           {interpMode=RCIM_Cubic, tangentMode=RCTM_User, time=0.4, value=0}
          ]
         }}})
     ))
    ScaleA=true
    ScaleRGB=true
   ) {enabled=false, module=ScaleColor001}
   &ScaleSpriteSize(
    Scale Sprite Size Mode=::NewEnumerator0
    Uniform Scale Factor=&Lerp_Float(
     A=&RandomRangeFloat(Minimum=3, Maximum=4)
     B=1
     Alpha=&FloatFromCurve(FloatCurve=DataInterface() {propertyValues {Curve {keys [
          {interpMode=RCIM_Cubic, leaveTangentWeight=0.8298, tangentMode=RCTM_User, time=0, value=0}
          {arriveTangentWeight=0.7442, interpMode=RCIM_Cubic, tangentMode=RCTM_User, time=0.5, value=1}
         ]}}})
    )
   ) {module=ScaleSpriteSize001}
   &ScaleSpriteSize(
    Scale Sprite Size Mode=::NewEnumerator0
    Uniform Scale Factor=&Lerp_Float(
     A=0.6
     B=0.85
     Alpha=&FloatFromCurve(
      FloatCurve=DataInterface() {propertyValues {Curve {keys [<at 0.1 0>,<at 0.3 1>,<at 1 1>]}}}
      CurveIndex=&FN_LOD_Distance_Remap(Start Distance=0, End Distance=5000, Minimum Output=0, Maximum Output=1, Clamp Output=true)
     )
    )
   )
   &DynamicMaterialParameters(
    Index 0 Param 1=&RandomRangeFloat(Minimum=0, Maximum=100)
    Index 0 Param 2=&Lerp_Float(
     A=-0.5
     B=0.3
     Alpha=&FloatFromCurve(FloatCurve=DataInterface() {propertyValues {Curve {keys [{interpMode=RCIM_Cubic, leaveTangent=2.5933, tangentMode=RCTM_Break, time=0, value=0}, <at 1 1>]}}})
    )
    Index 0 Param 3=&Lerp_Float(
     A=0
     B=2
     Alpha=&FloatFromCurve(FloatCurve=DataInterface() {propertyValues {Curve {
         defaultValue=1
         keys [
          {interpMode=RCIM_Cubic, tangentMode=RCTM_User, time=0, value=1}
          {arriveTangent=-1, interpMode=RCIM_Cubic, leaveTangent=-1, time=0.4, value=0.225}
          {interpMode=RCIM_Cubic, tangentMode=RCTM_User, time=1, value=0}
         ]
        }}})
    )
    Module.Index0Param0Write=true
    Module.Index0Param1Write=true
    Module.Index0Param2Write=true
    Module.Index0Param3Write=true
   )
   &FN_Sprite_Attributes(
    Color=<rgba 1 0.01 0.0265 0.05>
    Create Color Attribute=true
    Create SpriteSize Attribute=true
    SpriteSize=<366.7501 353.7109>
    Write to Module Namespace=true
   )
  ]
  rendererClasses [&NiagaraSpriteRendererProperties]
  renderers [{rendererClass=&NiagaraSpriteRendererProperties}]
 }
 Directional_Splashes {
  emitterUpdateScript [
   &EmitterState()
   &SpawnBurst_Instantaneous(
    Spawn Count=&RandomRangeInt(Minimum=6, Maximum=10)
    Spawn Probability=&MakeCustomFloatFromBool(Bool=$User.EnableGoo, True Float=1, False Float=0)
    Spawn Time=0.09
    Use Spawn Probability=true
    UseLoopCountLimit=false
   )
  ]
  particleSpawnScript [
   &InitializeParticle(
    Color Maximum=<rgba 0.3642 0.23 0.46 1>
    Color Minimum=<rgba 0.1507 0.1276 0.22 1>
    Color Mode=::NewEnumerator2
    Lifetime Max=0.65
    Lifetime Min=0.33
    Lifetime Mode=::NewEnumerator1
    Mesh Scale=&MakeVector(
     X=&Multiply_Float(A=&RandomRangeFloat(Minimum=0.5, Maximum=1), B=4)
     Y=1
     Z=1.25
    )
    Mesh Scale Mode=::NewEnumerator4
    Sprite Size=&Multiply_Vector2D_ByFloat(Float=0.2, Vector2D=&RandomRangeVector2D(Minimum=<150 200>, Maximum=<200 300>))
    Sprite Size Mode=::NewEnumerator4
    UsePositionOffset=false
    Write Material Random=true
   )
   &ShapeLocation(Hemisphere Distribution=<1 0.5>, MaskSpawnGroup=false, Override Seed=false, Sphere Radius=40, UseOwnerScale=true) {enabled=false, module=ShapeLocation001}
   &ShapeLocation(
    Hemisphere Distribution=<1 0.5>
    MaskSpawnGroup=false
    Override Seed=false
    Rotation Axis=<0 1 0>
    Rotation Mode=::NewEnumerator1
    Sphere Radius=60
    Sphere Surface Distribution=0.2
    UseOwnerScale=true
   )
   &AddVelocity(
    Module.Use Falloff Distance=false
    Origin Offset=&RandomRangeVector(Minimum=<0 -7 0>, Maximum=<0 -14 0>)
    Override Seed=false
    Velocity Mode=::NewEnumerator1
    Velocity Speed=100
   )
   &AddVelocity(Velocity=<0 0 30>, Velocity Speed Scale=&RandomRangeFloat(Minimum=0.66, Maximum=1)) {module=AddVelocity001}
  ]
  particleUpdateScript [
   &ParticleState()
   &ScaleMeshSize(Scale Factor=&VectorFromFloat(Value=&RandomRangeFloat(Minimum=0.3, Maximum=0.6)))
   &Drag(Drag=&RandomRangeFloat(Minimum=2, Maximum=5), Use Linear Drag=true, Use Rotational Drag=false)
   &GravityForce(Gravity=<0 0 -50>) {enabled=false}
   &SolveForcesAndVelocity(Clamp Velocity=false, Limit Acceleration=false)
   &DynamicMaterialParameters(
    Index 0 Param 1=&UniformAOrBFloat(A=0, Evaluation Type=::NewEnumerator0, Recalculate Random Each Loop=true, Override Seed=false)
    Index 0 Param 2=&FloatFromCurve(FloatCurve=DataInterface() {propertyValues {Curve {keys [
         {arriveTangent=1.5496, interpMode=RCIM_Cubic, leaveTangent=1.5496, tangentMode=RCTM_User, time=0, value=0.2}
         {arriveTangent=0.3249, interpMode=RCIM_Cubic, leaveTangent=0.3249, tangentMode=RCTM_User, time=1, value=1}
        ]}}})
    Index 0 Param 3=&FloatFromCurve(
     FloatCurve=DataInterface() {propertyValues {Curve {keys [
         {arriveTangent=0.2226, interpMode=RCIM_Cubic, leaveTangent=0.2226, tangentMode=RCTM_User, time=0, value=0}
         {arriveTangent=1.6484, interpMode=RCIM_Cubic, leaveTangent=1.6484, tangentMode=RCTM_User, time=1, value=1}
        ]}}}
     Scale Curve=10
    )
    Module.Index0Param0Write=true
    Module.Index0Param1Write=true
    Module.Index0Param2Write=true
    Module.Index0Param3Write=true
   ) {module=DynamicMaterialParameters002}
   &ScaleColor(
    Scale Alpha=&FloatFromCurve(
     FloatCurve=DataInterface() {propertyValues {Curve {keys [
         {arriveTangent=-0.1364, interpMode=RCIM_Cubic, leaveTangent=-0.1364, tangentMode=RCTM_User, time=0.25, value=1}
         {arriveTangent=-3.5002, interpMode=RCIM_Cubic, leaveTangent=-3.5002, tangentMode=RCTM_User, time=1, value=0}
        ]}}}
     Scale Curve=&RandomRangeFloat(Minimum=0.8, Maximum=1.5)
    )
    Scale RGB=&MakeVectorFromLinearColorRGB(Color=&ColorFromCurve(ColorCurve=DataInterface() {propertyValues {
        AlphaCurve {keys [<at 0 1 RCIM_Cubic>]}
        BlueCurve {keys [<at 0 6 RCIM_Cubic>,<at 0.2324 1>,<at 1 0.5 RCIM_Cubic>]}
        GreenCurve {keys [<at 0 0.36 RCIM_Cubic>,<at 0.2324 0.74>,<at 1 0.35 RCIM_Cubic>]}
        RedCurve {keys [<at 0 3.18 RCIM_Cubic>,<at 0.2324 0.8267>,<at 1 0.35 RCIM_Cubic>]}
       }}))
    ScaleA=true
    ScaleRGB=true
   )
   &ScaleColor(
    Scale Alpha=&FloatFromCurve(
     FloatCurve=DataInterface() {propertyValues {Curve {keys [
         {arriveTangent=-0.1364, interpMode=RCIM_Cubic, leaveTangent=-0.1364, tangentMode=RCTM_User, time=0.25, value=1}
         {arriveTangent=-3.5002, interpMode=RCIM_Cubic, leaveTangent=-3.5002, tangentMode=RCTM_User, time=1, value=0}
        ]}}}
     Scale Curve=&RandomRangeFloat(Minimum=0.8, Maximum=1)
    )
    Scale RGB=&MakeVectorFromLinearColorRGB(Color=&ColorFromCurve(ColorCurve=DataInterface() {propertyValues {
        AlphaCurve {keys [<at 0 1 RCIM_Cubic>]}
        BlueCurve {keys [<at 0 6 RCIM_Cubic>,<at 0.2324 1>,<at 1 0.5 RCIM_Cubic>]}
        GreenCurve {keys [<at 0 0.36 RCIM_Cubic>,<at 0.2324 0.74>,<at 1 0.35 RCIM_Cubic>]}
        RedCurve {keys [<at 0 3.18 RCIM_Cubic>,<at 0.2324 0.8267>,<at 1 0.35 RCIM_Cubic>]}
       }}))
    ScaleA=true
    ScaleRGB=true
   ) {enabled=false, module=ScaleColor001}
   /SetVariables(Particles.PositionWaves:NiagaraPosition=&AddVectorToPosition(
     Position=$Particles.Position
     Vector=&MakeVector(
      X=0
      Y=0
      Z=&Multiply_Float(
       A=&RandomRangeFloat(Minimum=-30, Maximum=-70)
       B=&FloatFromCurve(FloatCurve=DataInterface() {propertyValues {Curve {keys [
            {arriveTangent=0.0456, interpMode=RCIM_Cubic, leaveTangent=0.0456, tangentMode=RCTM_User, time=0, value=0}
            {arriveTangent=2.1323, interpMode=RCIM_Cubic, leaveTangent=2.1323, tangentMode=RCTM_User, time=1, value=1}
           ]}}})
      )
     )
    )) {isSetParameters=true}
   &ScaleColor(
    Scale RGB=&VectorFromFloat(Value=&RandomRangeFloat(Minimum=1.5, Maximum=4))
    ScaleA=false
    ScaleRGB=true
   ) {module=ScaleColor002}
  ]
  rendererClasses [&NiagaraMeshRendererProperties]
  renderers [
   {rendererClass=&NiagaraMeshRendererProperties, rendererIndex=0}
   {rendererClass=&NiagaraMeshRendererProperties, rendererIndex=1}
  ]
 }
 Drops {
  emitterUpdateScript [
   &EmitterState()
   &SpawnBurst_Instantaneous(
    Spawn Count=&RandomRangeInt(Minimum=22, Maximum=25)
    Spawn Probability=&MakeCustomFloatFromBool(Bool=$User.EnableGoo, True Float=1, False Float=0)
    Spawn Time=0.11
    Use Spawn Probability=true
   )
  ]
  particleSpawnScript [
   &InitializeParticle(
    Color=&RandomRangeLinearColor(Minimum=<rgba 0.8 0.8 0.8 1>, Maximum=<rgba 1 1 1 1>)
    Lifetime=&UniformRangedFloat(Minimum=0.4, Maximum=1.4)
    Mass Mode=::NewEnumerator1
    Sprite Size=&Multiply_Vector2D_ByFloat(
     Float=&UniformAOrBFloat(
      A=&RandomRangeFloat(Minimum=2, Maximum=8)
      B=8
      Distribution Weight=0.1
      Evaluation Type=::NewEnumerator0
      Recalculate Random Each Loop=true
      Override Seed=false
     )
     Vector2D=&UniformRangedVector2D(Minimum=<2 3>, Maximum=<2 4>)
    )
    Sprite Size Mode=::NewEnumerator4
    UsePositionOffset=false
    Write Material Random=true
   ) {module=InitializeParticle001}
   &ShapeLocation(
    Hemisphere Distribution=<1 0.5>
    MaskSpawnGroup=false
    Override Seed=false
    Rotation Axis=<0 1 0>
    Rotation Mode=::NewEnumerator1
    Sphere Radius=60
    Sphere Surface Distribution=0.631
    UseOwnerScale=true
   )
   &AddVelocityFromPoint(
    Module.Use Falloff Distance=false
    Origin Offset=<0 -5 0>
    Velocity Strength=&Multiply_Float(A=&UniformRangedFloat(Minimum=550, Maximum=1150), B=1)
   )
   &self:ColorFromTOD_02(
    Afternoon:LinearColor=<rgba 0.5122 0.5122 0.5122 1>
    Morning:LinearColor=<rgba 0.4948 0.4948 0.4948 1>
    Night:LinearColor=<rgba 0.1119 0.152 0.2378 1>
    TOD:Vector4f=&MakeVector4_FromLinearColor(LinearColor=$NPC.FortniteNPC.FortniteActiveTimeOfDay)
   ) {enabled=false, module=ColorFromTOD_02}
   /Niagara/Modules/Update/SubUV/SubUVAnimation(SubUV Animation Mode=::NewEnumerator1) {module=SubUVAnimation001}
  ]
  particleUpdateScript [
   &ParticleState()
   &ScaleSpriteSize(Uniform Curve Sprite Scale=DataInterface() {propertyValues {Curve {keys [
        {arriveTangent=8.1474, interpMode=RCIM_Cubic, leaveTangent=8.1474, tangentMode=RCTM_User, time=0, value=0}
        {arriveTangentWeight=0.0928, interpMode=RCIM_Cubic, leaveTangentWeight=0.5258, tangentMode=RCTM_User, time=0.15, value=1}
        {arriveTangent=-2.0068, arriveTangentWeight=0.296, interpMode=RCIM_Cubic, leaveTangent=-2.0068, tangentMode=RCTM_User, time=1, value=0}
       ]}}}) {module=ScaleSpriteSize001}
   &Drag(Drag=&RandomRangeFloat(Minimum=2, Maximum=3), Use Linear Drag=true, Use Rotational Drag=true)
   &GravityForce(Gravity=&MakeVector(X=0, Y=0, Z=&RandomRangeFloat(Minimum=-700, Maximum=-980))) {module=GravityForce001}
   &SolveForcesAndVelocity(Clamp Velocity=false, Limit Acceleration=false)
   &ScaleSpriteSizeBySpeed(Max Scale Factor=<0.66 4>, Min Scale Factor=<1 0.5>)
   &ScaleColor(
    Scale Alpha=&FloatFromCurve(
     FloatCurve=DataInterface() {propertyValues {Curve {keys [
         {arriveTangent=8.1018, interpMode=RCIM_Cubic, leaveTangent=8.1018, tangentMode=RCTM_User, time=0, value=0}
         {arriveTangentWeight=0.0573, interpMode=RCIM_Cubic, leaveTangentWeight=0.3728, tangentMode=RCTM_User, time=0.1, value=1}
         {arriveTangentWeight=0.2167, interpMode=RCIM_Cubic, leaveTangentWeight=0.0833, tangentMode=RCTM_User, time=0.75, value=1}
         <at 1 0 RCIM_Cubic>
        ]}}}
     Scale Curve=3
    )
    Scale RGB=&MakeVectorFromLinearColorRGB(Color=&ColorFromCurve(ColorCurve=DataInterface() {propertyValues {
        AlphaCurve {keys [<at 0 1 RCIM_Cubic>]}
        BlueCurve {keys [<at 0 3>, {arriveTangent=-2.34, interpMode=RCIM_Cubic, leaveTangent=-2.34, time=0.1504, value=0.9}, <at 1 0.66 RCIM_Cubic>]}
        GreenCurve {keys [<at 0 0.69>, {arriveTangent=-0.492, interpMode=RCIM_Cubic, leaveTangent=-0.492, time=0.1504, value=0.243}, <at 1 0.198 RCIM_Cubic>]}
        RedCurve {keys [<at 0 2.076>, {arriveTangent=-1.8395, interpMode=RCIM_Cubic, leaveTangent=-1.8395, time=0.1504, value=0.6372}, <at 1 0.2365 RCIM_Cubic>]}
       }}))
    ScaleA=true
    ScaleRGB=true
   )
   &DynamicMaterialParameters(
    Index 0 Param 1=&RandomRangeFloat(Minimum=0, Maximum=1)
    Index 0 Param 2=&FloatFromCurve(FloatCurve=DataInterface() {propertyValues {Curve {keys [<at 0.35 0.45>,<at 1 0.9>]}}})
    Index 0 Param 3=&FloatFromCurve(FloatCurve=DataInterface() {propertyValues {Curve {keys [{arriveTangent=2.6651, interpMode=RCIM_Cubic, leaveTangent=2.6651, tangentMode=RCTM_User, time=0, value=0}, <at 1 1 RCIM_Cubic>]}}})
    Index 0 Param 4=&FloatFromCurve(FloatCurve=DataInterface() {propertyValues {Curve {keys [<at 0 0 RCIM_Cubic>, {arriveTangent=2.332, interpMode=RCIM_Cubic, leaveTangent=2.332, tangentMode=RCTM_User, time=1, value=1}]}}})
    Module.Param0WriteEnabled=true
    Module.Param1WriteEnabled=true
    Module.Param2WriteEnabled=false
    Module.Param3WriteEnabled=true
   ) {module=DynamicMaterialParameters001}
  ]
  rendererClasses [&NiagaraSpriteRendererProperties]
  renderers [{rendererClass=&NiagaraSpriteRendererProperties}]
 }
 LightStreaks {
  emitterUpdateScript [
   &EmitterState(Life Cycle Mode=::NewEnumerator1, Loop Behavior=::NewEnumerator1, Loop Duration=0.16, UseLoopDelay=false)
   &SpawnBurst_Instantaneous(Spawn Count=7, Use Spawn Probability=false, UseLoopCountLimit=false)
  ]
  particleSpawnScript [
   /SetVariables(Particles.RandomFloat=&UniformRangedFloat(Minimum=0, Maximum=3)) {isSetParameters=true}
   &InitializeParticle(
    Color=&HueShiftLinearColor(
     Hue=0
     Saturation=1
     Value=&RandomRangeFloat(Minimum=0.1, Maximum=1)
     Linear Color=<rgba 0.3176 0.0047 0.474 3>
    )
    Lifetime Max=0.24
    Lifetime Min=0.19
    Lifetime Mode=::NewEnumerator1
    Sprite Size Max=<30 22>
    Sprite Size Min=<20 12>
    Sprite Size Mode=::NewEnumerator2
    UsePositionOffset=false
    Write Material Random=true
   )
   &ShapeLocation(
    Hemisphere Distribution=<1 0.5>
    MaskSpawnGroup=false
    Override Seed=false
    Rotation Angle=0.25
    Rotation Axis=<0 1 0>
    Rotation Mode=::NewEnumerator1
    Sphere Radius=17
    Sphere Surface Distribution=0.4881
    UseOwnerScale=true
   )
   &AddVelocity(
    Module.Use Falloff Distance=false
    Origin Offset=<0 -5 0>
    Override Seed=false
    Velocity Mode=::NewEnumerator1
    Velocity Speed=&Multiply_Float(
     A=&Lerp_Float(
      A=500
      B=1000
      Alpha=&FloatFromCurve(
       FloatCurve=DataInterface() {propertyValues {Curve {keys [
           {interpMode=RCIM_Cubic, leaveTangent=2.3303, tangentMode=RCTM_Break, time=0, value=0}
           {interpMode=RCIM_Cubic, tangentMode=RCTM_Break, time=1, value=1}
          ]}}}
       CurveIndex=&RandomRangeFloat(Minimum=0, Maximum=1)
      )
     )
     B=&Lerp_Float(
      A=0.6
      B=1
      Alpha=&FloatFromCurve(
       FloatCurve=DataInterface() {propertyValues {Curve {keys [<at 0.1 0>,<at 0.3 1>,<at 1 1>]}}}
       CurveIndex=&FN_LOD_Distance_Remap(Start Distance=0, End Distance=5000, Minimum Output=0, Maximum Output=1, Clamp Output=true)
      )
     )
    )
   ) {module=AddVelocity001}
   /Niagara/Modules/Update/SubUV/V2/SubUVAnimation(End Frame=3, SubUV Animation Mode:ENiagara_SubUVLookupMode=::NewEnumerator2)
   &CameraOffset(Camera Offset Amount=-80)
  ]
  particleUpdateScript [
   &ParticleState() {module=UpdateAge}
   &Drag(Drag=&RandomRangeFloat(Minimum=1, Maximum=4), Use Linear Drag=true, Use Rotational Drag=true)
   &SolveForcesAndVelocity(Clamp Velocity=false, Limit Acceleration=false)
   &ScaleSpriteSize(Scale Factor=&Vector2DFromCurve(
     Vector2Curve=DataInterface() {propertyValues {
       XCurve {keys [
         {interpMode=RCIM_Cubic, tangentMode=RCTM_User, time=0, value=3}
         {arriveTangent=-1.2194, interpMode=RCIM_Cubic, leaveTangent=-1.2194, tangentMode=RCTM_User, time=0.25, value=1.2}
         {interpMode=RCIM_Cubic, tangentMode=RCTM_User, time=1, value=0.5}
        ]}
       YCurve {keys [
         {arriveTangent=2.3946, interpMode=RCIM_Cubic, leaveTangent=2.3946, tangentMode=RCTM_User, time=0, value=0.66}
         {arriveTangent=1.2477, interpMode=RCIM_Cubic, leaveTangent=1.2477, tangentMode=RCTM_User, time=1, value=3.7}
        ]}
      }}
     ScaleCurve=<0.75 2>
    ))
   &ScaleSpriteSize(
    Scale Sprite Size Mode=::NewEnumerator0
    Uniform Scale Factor=&Lerp_Float(
     A=1
     B=3
     Alpha=&FloatFromCurve(
      FloatCurve=DataInterface() {propertyValues {Curve {keys [<at 0.1 0>,<at 0.3 1>,<at 1 1>]}}}
      CurveIndex=&FN_LOD_Distance_Remap(Start Distance=0, End Distance=5000, Minimum Output=0, Maximum Output=1, Clamp Output=true)
     )
    )
   ) {module=ScaleSpriteSize001}
   &ScaleColor(
    Scale Alpha=&FloatFromCurve(
     FloatCurve=DataInterface() {propertyValues {Curve {keys [
         <at 0 0 RCIM_Cubic>
         {interpMode=RCIM_Cubic, tangentMode=RCTM_Break, time=0.11, value=1}
         {interpMode=RCIM_Cubic, tangentMode=RCTM_Break, time=1, value=0}
        ]}}}
     Scale Curve=&Multiply_Float(
      A=&RandomRangeFloat(Minimum=0.05, Maximum=0.2)
      B=&FNTimeOfDayScalar(Night Scale=0.3, Morning Scale=0.5, Evening Scale=0.5)
     )
    )
    Scale RGB=&VectorFromFloat(Value=&Multiply_Float(
      A=&RandomRangeFloat(Minimum=5, Maximum=15)
      B=&FloatFromCurve(
       FloatCurve=DataInterface() {propertyValues {Curve {keys [<at 0 1>,<at 0.15 0.9873>,<at 0.3816 0.1931>,<at 1 0>]}}}
       Scale Curve=&FNTimeOfDayScalar(Night Scale=2, Morning Scale=3, Day Scale=10, Evening Scale=3)
      )
     ))
    ScaleA=true
    ScaleRGB=true
   )
   &DynamicMaterialParameters(
    Index 0 Param 2=1
    Index 0 Param 3=1
    Index 0 Param 4=&FloatFromCurve(FloatCurve=DataInterface() {propertyValues {Curve {keys [<at 0.3 1>,<at 1 0.25>]}}})
    Module.Param0WriteEnabled=true
    Module.Param1WriteEnabled=true
    Module.Param2WriteEnabled=true
    Module.Param3WriteEnabled=true
   )
   /SetVariables(Particles.ColorDarkStreaks:LinearColor=&MultiplyLinearColor(
     Linear Color=$Particles.Color
     Scale Factor=&MakeVector4_FromLinearColor(LinearColor=&ColorFromCurve(ColorCurve=DataInterface() {propertyValues {
         AlphaCurve {keys [<at 0 10 RCIM_Cubic>]}
         BlueCurve {keys [<at 0 0.25 RCIM_Cubic>,<at 0.5792 0.01 RCIM_Cubic>]}
         GreenCurve {keys [<at 0 0.125 RCIM_Cubic>,<at 0.5792 0.0058 RCIM_Cubic>]}
         RedCurve {keys [<at 0 0.1562 RCIM_Cubic>,<at 0.5792 0.005 RCIM_Cubic>]}
        }}))
    )) {isSetParameters=true}
  ]
  rendererClasses [&NiagaraSpriteRendererProperties]
  renderers [
   {rendererClass=&NiagaraSpriteRendererProperties}
   {rendererClass=&NiagaraSpriteRendererProperties, rendererIndex=1}
  ]
 }
 MergedGlow {
  emitterSpawnScript [/SetVariables() {isSetParameters=true}]
  emitterUpdateScript [
   &EmitterState()
   &SpawnBurst_Instantaneous(Spawn Count=1, Use Spawn Probability=false)
  ]
  particleSpawnScript [&InitializeParticle(
    Color=<rgba 0.5787 0 1 1>
    Lifetime=0.16
    Sprite Size Mode=::NewEnumerator3
    Uniform Sprite Size=&Multiply_Float(A=900, B=1)
    UsePositionOffset=false
    Write Material Random=true
   )]
  particleUpdateScript [
   &ParticleState()
   &ScaleSpriteSize(
    Scale Sprite Size Mode=::NewEnumerator0
    Uniform Scale Factor=&Lerp_Float(
     A=0.4
     B=1
     Alpha=&FloatFromCurve(
      FloatCurve=DataInterface() {propertyValues {Curve {keys [<at 0.1 0>,<at 0.3 1>,<at 1 1>]}}}
      CurveIndex=&FN_LOD_Distance_Remap(Start Distance=0, End Distance=5000, Minimum Output=0, Maximum Output=1, Clamp Output=true)
     )
    )
   )
   &CameraOffset(Camera Offset Amount=50)
   &ScaleSpriteSize(
    Scale Sprite Size Mode=::NewEnumerator0
    Uniform Scale Factor=&Lerp_Float(
     A=0
     B=1
     Alpha=&FloatFromCurve(FloatCurve=DataInterface() {propertyValues {Curve {
         defaultValue=0
         keys [
          {arriveTangent=9.0725, interpMode=RCIM_Cubic, leaveTangent=9.0725, tangentMode=RCTM_User, time=0.0046, value=0.1473}
          {arriveTangent=1.4412, interpMode=RCIM_Cubic, leaveTangent=1.4412, tangentMode=RCTM_User, time=0.1, value=0.7}
          {arriveTangent=-0.0405, interpMode=RCIM_Cubic, leaveTangent=-0.0405, tangentMode=RCTM_User, time=0.5, value=0.95}
         ]
        }}})
    )
   ) {module=ScaleSpriteSize001}
   &ScaleColor(
    Scale Alpha=&FloatFromCurve(FloatCurve=DataInterface() {propertyValues {Curve {keys [<at 0.5 0.95 RCIM_Cubic>, {arriveTangent=-1.9, interpMode=RCIM_Cubic, leaveTangent=-1.9, time=0.646, value=0.2596}, <at 1 0 RCIM_Cubic>]}}})
    Scale RGB=&VectorFromFloat(Value=&Lerp_Float(
      A=0
      B=2
      Alpha=&FloatFromCurve(FloatCurve=DataInterface() {propertyValues {Curve {
          defaultValue=1
          keys [
           {arriveTangent=9.0725, interpMode=RCIM_Cubic, leaveTangent=9.0725, tangentMode=RCTM_User, time=0, value=0.1}
           {arriveTangent=-2.5249, interpMode=RCIM_Cubic, leaveTangent=-2.5249, tangentMode=RCTM_User, time=0.1118, value=0.9917}
           {arriveTangent=-1.0039, interpMode=RCIM_Cubic, leaveTangent=-1.0039, time=0.2507, value=0.4212}
           {arriveTangent=-0.0405, interpMode=RCIM_Cubic, leaveTangent=-0.0405, tangentMode=RCTM_User, time=1, value=0.1}
          ]
         }}})
     ))
    ScaleA=true
    ScaleRGB=true
   )
   /SetVariables(
    Particles.FlareCenterColor:LinearColor=&MultiplyLinearColor(
     Linear Color=$Particles.Initial.Color
     Scale Factor=&MakeVector4FromVectorAndFloat(
      "Vector (RGB)"=&VectorFromFloat(Value=5)
      "Float (Alpha)"=&FloatFromCurve(FloatCurve=DataInterface() {propertyValues {Curve {keys [<at 0.5 0.95 RCIM_Cubic>, {arriveTangent=-1.9, interpMode=RCIM_Cubic, leaveTangent=-1.9, time=0.646, value=0.2596}, <at 1 0 RCIM_Cubic>]}}})
     )
    )
    Particles.FlareCenterSize:Vector2f=&Multiply_Vector2D_ByFloat(Float=0.5, Vector2D=$Particles.SpriteSize)
   ) {isSetParameters=true}
   /SetVariables(
    Particles.ColorDarkGlow:LinearColor=&MultiplyLinearColor(
     Linear Color=$Particles.Color
     Scale Factor=&MakeVector4FromVectorAndFloat("Vector (RGB)"=&VectorFromFloat(Value=0.1), "Float (Alpha)"=1.2)
    )
    Particles.SpriteSizeDarkGlow:Vector2f=&Multiply_Vector2D_ByFloat(Float=1.5, Vector2D=$Particles.SpriteSize)
   ) {isSetParameters=true}
  ]
  rendererClasses [&NiagaraSpriteRendererProperties]
  renderers [
   {rendererClass=&NiagaraSpriteRendererProperties}
   {rendererClass=&NiagaraSpriteRendererProperties, rendererIndex=1}
   {rendererClass=&NiagaraSpriteRendererProperties, rendererIndex=2}
  ]
 }
 NE_Splash_Directional {
  emitterUpdateScript [
   &EmitterState()
   &SpawnBurst_Instantaneous(
    Spawn Count=&RandomRangeInt(Minimum=8, Maximum=15)
    Spawn Time=0.05
    Use Spawn Probability=false
    UseLoopCountLimit=false
   )
  ]
  particleSpawnScript [
   &InitializeParticle(
    Color=<rgba 0.5 0.5 0.5 1>
    Lifetime Max=1
    Lifetime Min=0.4
    Lifetime Mode=::NewEnumerator1
    Sprite Rotation Mode=::NewEnumerator2
    Sprite Size Max=<300 600>
    Sprite Size Min=<125 400>
    Sprite Size Mode=::NewEnumerator2
    Sprite UV Mode=::NewEnumerator3
    UsePositionOffset=false
    Write Material Random=true
   )
   &ShapeLocation(
    MaskSpawnGroup=false
    Rotation Axis=<0 1 0>
    Rotation Mode=::NewEnumerator1
    Sphere Distribution=::NewEnumerator1
    Sphere Radius=30
    U Position=&ReturnNormalizedExecIndex()
    UseOwnerScale=true
    V Position=&RandomRangeFloat(Minimum=0.6, Maximum=0.8)
   )
   &AddVelocity(
    Velocity=$Particles.ShapeLocation.ShapeNormal
    Velocity Speed Scale=&RandomRangeFloat(Minimum=550, Maximum=650)
   ) {module=AddVelocity002}
   &SubUVAnimation(
    End Frame Range Override=3
    "Sprite Renderer "=DataInterface() {propertyValues {SpriteRenderer=&self:NE_Splash_Directional_6.NiagaraSpriteRendererProperties_0}}
    SubUV Animation Mode:ENiagara_SubUVLookupModeV2=::NewEnumerator2
    UseEndFrame=true
    UseStartFrame=false
   )
  ]
  particleUpdateScript [
   &ParticleState()
   &SpriteRotationRate(Rotation Rate=&RandomRangeFloat(Minimum=-90, Maximum=90))
   &ScaleSpriteSize(
    Uniform Curve Scale=0.85
    Uniform Curve Sprite Scale=DataInterface() {propertyValues {Curve {keys [{interpMode=RCIM_Cubic, leaveTangent=3.7005, tangentMode=RCTM_Break, time=0, value=0}, {tangentMode=RCTM_User, time=0.5, value=0.95}]}}}
   )
   &GravityForce()
   &SolveForcesAndVelocity(Clamp Velocity=false, Limit Acceleration=false)
   &ScaleColor(
    Linear Color Curve=DataInterface() {propertyValues {
      AlphaCurve {keys [
        {arriveTangent=-2.2019, interpMode=RCIM_Cubic, leaveTangent=-2.2019, tangentMode=RCTM_User, time=0, value=1}
        {arriveTangent=-0.1268, interpMode=RCIM_Cubic, leaveTangent=-0.1268, tangentMode=RCTM_User, time=1, value=0}
       ]}
      BlueCurve {keys [<at 0 4>,<at 0.5225 0.9479>]}
      GreenCurve {keys [<at 0 0.24>,<at 0.5225 0.0643>]}
      RedCurve {keys [<at 0 2.12>,<at 0.5225 0.2115>]}
     }}
    Scale Mode=::NewEnumerator2
   )
   &DynamicMaterialParameters(Module.Index0Float0Write=false, Module.Index0Float1Write=true, Module.Index0Float2Write=false, Module.Index0Float3Write=false)
  ]
  rendererClasses [&NiagaraSpriteRendererProperties]
  renderers [{rendererClass=&NiagaraSpriteRendererProperties}]
 }
 Rainbow_Smoke {
  emitterUpdateScript [
   &EmitterState()
   &SpawnBurst_Instantaneous(Spawn Count=2, Use Spawn Probability=false, UseLoopCountLimit=false)
  ]
  particleSpawnScript [
   &InitializeParticle(
    Color=&HueShiftLinearColor(
     Hue=0
     Saturation=1
     Value=&RandomRangeFloat(Minimum=0.5, Maximum=1)
     Linear Color=<rgba 0.0145 0.0054 0.0312 1>
    )
    Lifetime Min=1.5
    Lifetime Mode=::NewEnumerator1
    Sprite Rotation Mode=::NewEnumerator2
    Sprite Size Mode=::NewEnumerator3
    Uniform Sprite Size=&Multiply_Float(A=&RandomRangeFloat(Minimum=750, Maximum=850), B=$User.FXScale)
    UsePositionOffset=false
    Write Material Random=true
   )
   &DynamicMaterialParameters(
    Index 0 Param 1=&RandomRangeFloat(Minimum=0, Maximum=1)
    Module.Param0WriteEnabled=true
    Module.Param1WriteEnabled=true
    Module.Param2WriteEnabled=true
    Module.Param3WriteEnabled=true
   )
   &AddVelocity(Velocity=<0 500 0>)
   &ShapeLocation(
    Hemisphere Distribution=<1 0.5>
    MaskSpawnGroup=false
    Offset=<0 0 15>
    Offset Mode=::NewEnumerator0
    Override Seed=false
    Sphere Radius=50
    Sphere Surface Distribution=0.4286
    UseOwnerScale=true
   )
   &CameraOffset(Camera Offset Amount=70)
   /SetVariables(Particles.Color=&MultiplyLinearColor(
     Linear Color=$Particles.Color
     Scale Factor=&MakeVector4FromVectorAndFloat("Vector (RGB)"=<1 1 1>, "Float (Alpha)"=1)
    )) {isSetParameters=true}
  ]
  particleUpdateScript [
   &ParticleState()
   &ScaleSpriteSize(
    Uniform Curve Scale=&Lerp_Float(
     A=0.4
     B=1
     Alpha=&FloatFromCurve(
      FloatCurve=DataInterface() {propertyValues {Curve {keys [<at 0.1 0>,<at 0.3 1>,<at 1 1>]}}}
      CurveIndex=&FN_LOD_Distance_Remap(Start Distance=0, End Distance=5000, Minimum Output=0, Maximum Output=1, Clamp Output=true)
     )
    )
    Uniform Curve Sprite Scale=DataInterface() {propertyValues {Curve {keys [
        {arriveTangent=-0.077, interpMode=RCIM_Cubic, leaveTangent=-0.077, tangentMode=RCTM_User, time=0, value=0.2}
        {arriveTangent=0.8, interpMode=RCIM_Cubic, leaveTangent=0.8, time=0.15, value=0.6}
        {arriveTangent=0.1164, interpMode=RCIM_Cubic, leaveTangent=0.1164, tangentMode=RCTM_User, time=1, value=1}
       ]}}}
   )
   &Drag(Drag=5, Use Linear Drag=true, Use Rotational Drag=true)
   &SolveForcesAndVelocity(Clamp Velocity=false, Limit Acceleration=false)
   &ScaleColor(
    Scale Alpha=&Multiply_Float(
     A=&FloatFromCurve(FloatCurve=DataInterface() {propertyValues {Curve {keys [
          {arriveTangent=16.5662, interpMode=RCIM_Cubic, leaveTangent=16.5662, tangentMode=RCTM_User, time=0, value=0}
          {arriveTangent=-3.0691, interpMode=RCIM_Cubic, leaveTangent=-3.0691, tangentMode=RCTM_User, time=0.1, value=1}
          {arriveTangent=-0.231, arriveTangentWeight=0.7021, interpMode=RCIM_Cubic, leaveTangent=-0.231, tangentMode=RCTM_User, time=1, value=0}
         ]}}})
     B=1.5
    )
    Scale RGB=&VectorFromFloat(Value=&Multiply_Float(
      A=&RandomRangeFloat(Minimum=2, Maximum=3)
      B=&Lerp_Float(
       A=0.4
       B=3
       Alpha=&FloatFromCurve(FloatCurve=DataInterface() {propertyValues {Curve {keys [
            {arriveTangent=-5.0642, interpMode=RCIM_Cubic, leaveTangent=-5.0642, tangentMode=RCTM_User, time=0, value=1}
            {arriveTangent=-1.0691, interpMode=RCIM_Cubic, leaveTangent=-1.0691, tangentMode=RCTM_User, time=0.11, value=0.25}
            {interpMode=RCIM_Cubic, tangentMode=RCTM_User, time=0.7, value=0}
           ]}}})
      )
     ))
    ScaleA=true
    ScaleRGB=true
   )
  ]
  rendererClasses [&NiagaraSpriteRendererProperties]
  renderers [{rendererClass=&NiagaraSpriteRendererProperties}]
 }
 Splash_Center {
  emitterSpawnScript [/SetVariables(Emitter.LightExpornent=2, Emitter.LightPos:NiagaraPosition=<0 0 75>) {isSetParameters=true}]
  emitterUpdateScript [
   &EmitterState(Inactive Response=::NewEnumerator1, Life Cycle Mode=::NewEnumerator1, Loop Behavior=::NewEnumerator1, Loop Duration=0.3, UseLoopDelay=false)
   &SpawnBurst_Instantaneous(
    Spawn Count=2
    Spawn Probability=&MakeCustomFloatFromBool(Bool=$User.EnableGoo, True Float=1, False Float=0)
    Spawn Time=0.01
    Use Spawn Probability=true
    UseLoopCountLimit=false
   )
   /SetVariables(
    Emitter.EmitterRandom:NiagaraFloat=&RandomRangeFloat(Minimum=0, Maximum=1)
    Emitter.LightColor:LinearColor=&Multiply_LinearColorByFloat(
     Linear Color=<rgba 0.2753 0 1 1>
     Float=&FloatFromCurve(
      FloatCurve=DataInterface() {propertyValues {Curve {keys [
          {interpMode=RCIM_Cubic, leaveTangentWeight=0.9776, tangentMode=RCTM_Break, time=0.03, value=1}
          {arriveTangent=-2.4571, interpMode=RCIM_Cubic, tangentMode=RCTM_Break, time=0.225, value=0}
         ]}}}
      CurveIndex=$Emitter.Age
      Scale Curve=9
     )
    )
    Emitter.LightRadius:NiagaraFloat=&FloatFromCurve(FloatCurve=DataInterface() {propertyValues {Curve {keys [<at 0.2 1>,<at 0.225 0>]}}}, CurveIndex=$Emitter.Age, Scale Curve=500)
   ) {isSetParameters=true}
  ]
  particleSpawnScript [
   &InitializeParticle(
    Color=&RandomRangeLinearColor(Minimum=<rgba 0.8108 0.747 0.9 1>, Maximum=<rgba 0.8126 0.801 0.9 1>)
    Lifetime=&RandomRangeFloat(Minimum=0.5, Maximum=0.6)
    Sprite Rotation Mode=::NewEnumerator2
    Sprite Size Mode=::NewEnumerator1
    Uniform Sprite Size Max=135
    Uniform Sprite Size Min=100
    UsePositionOffset=false
    Write Material Random=true
   )
   &CameraOffset(Camera Offset Amount=10)
   &ShapeLocation(MaskSpawnGroup=false, Override Seed=false, Sphere Radius=7, Sphere Surface Distribution=0.4286, UseOwnerScale=true)
   &AddVelocity(Velocity=<0 350 0>, Velocity Speed Scale=&RandomRangeFloat(Minimum=0.4, Maximum=1))
  ]
  particleUpdateScript [
   &ParticleState()
   &SpriteRotationRate(Rotation Rate=&RandomRangeFloat(Minimum=-20, Maximum=20))
   &DynamicMaterialParameters(
    Index 0 Param 1=$Emitter.EmitterRandom
    Index 0 Param 2=&FloatFromCurve(FloatCurve=DataInterface() {propertyValues {Curve {
        defaultValue=1
        keys [
         {arriveTangent=0.1559, interpMode=RCIM_Cubic, leaveTangent=0.1559, tangentMode=RCTM_User, time=0.2, value=0}
         {arriveTangent=1.8635, interpMode=RCIM_Cubic, leaveTangent=1.8635, tangentMode=RCTM_User, time=0.4, value=0.5}
         {arriveTangent=0.4386, interpMode=RCIM_Cubic, leaveTangent=0.4386, tangentMode=RCTM_User, time=1, value=1}
        ]
       }}})
    Index 0 Param 4=&Lerp_Float(
     A=&RandomRangeFloat(Minimum=4, Maximum=6)
     B=1
     Alpha=&FloatFromCurve(FloatCurve=DataInterface() {propertyValues {Curve {keys [
          {interpMode=RCIM_Cubic, tangentMode=RCTM_User, time=0, value=0}
          {arriveTangent=3.8492, interpMode=RCIM_Cubic, leaveTangent=3.8492, tangentMode=RCTM_User, time=0.26, value=0.3}
          {arriveTangent=0.9756, interpMode=RCIM_Cubic, leaveTangent=0.9756, tangentMode=RCTM_User, time=0.5, value=1}
         ]}}})
    )
    Module.Param0WriteEnabled=true
    Module.Param1WriteEnabled=true
    Module.Param2WriteEnabled=true
    Module.Param3WriteEnabled=true
   ) {module=DynamicMaterialParameters001}
   &GravityForce(Gravity=&Multiply_VectorByFloat(Vector=<0 0 -980>, Float=&RandomRangeFloat(Minimum=0.6, Maximum=1)))
   &Drag(Use Linear Drag=true, Use Rotational Drag=true)
   &SolveForcesAndVelocity(Clamp Velocity=false, Limit Acceleration=false)
   &ScaleColor(
    Scale Alpha=&FloatFromCurve(
     FloatCurve=DataInterface() {propertyValues {Curve {keys [
         {arriveTangent=17.6185, interpMode=RCIM_Cubic, leaveTangent=17.6185, tangentMode=RCTM_User, time=0, value=0}
         {arriveTangentWeight=0.0836, interpMode=RCIM_Cubic, leaveTangentWeight=0.3344, tangentMode=RCTM_User, time=0.07, value=1}
         {arriveTangent=-0.184, interpMode=RCIM_Cubic, leaveTangent=-0.184, leaveTangentWeight=0.6975, tangentMode=RCTM_User, time=0.34, value=1}
         {arriveTangentWeight=0.4545, interpMode=RCIM_Cubic, tangentMode=RCTM_User, time=1, value=0}
        ]}}}
     Scale Curve=&RandomRangeFloat(Minimum=2, Maximum=3)
    )
    Scale RGB=&MakeVectorFromLinearColorRGB(Color=&ColorFromCurve(ColorCurve=DataInterface() {propertyValues {
        AlphaCurve {keys [<at 0 1 RCIM_Cubic>]}
        BlueCurve {keys [<at 0 3.5 RCIM_Cubic>,<at 0.2811 0.85>,<at 0.9319 0.36 RCIM_Cubic>]}
        GreenCurve {keys [<at 0 0.595 RCIM_Cubic>,<at 0.2811 0.595>,<at 0.9319 0.1908 RCIM_Cubic>]}
        RedCurve {keys [<at 0 2.0475 RCIM_Cubic>,<at 0.2811 0.697>,<at 0.9319 0.2134 RCIM_Cubic>]}
       }}))
    ScaleA=true
    ScaleRGB=true
   )
   &ScaleColor(
    Scale Alpha=&FloatFromCurve(
     FloatCurve=DataInterface() {propertyValues {Curve {keys [
         {arriveTangent=17.6185, interpMode=RCIM_Cubic, leaveTangent=17.6185, tangentMode=RCTM_User, time=0, value=0}
         {arriveTangentWeight=0.0836, interpMode=RCIM_Cubic, leaveTangentWeight=0.3344, tangentMode=RCTM_User, time=0.07, value=1}
         {arriveTangent=-0.184, interpMode=RCIM_Cubic, leaveTangent=-0.184, leaveTangentWeight=0.6975, tangentMode=RCTM_User, time=0.34, value=1}
         {arriveTangentWeight=0.4545, interpMode=RCIM_Cubic, tangentMode=RCTM_User, time=1, value=0}
        ]}}}
     Scale Curve=&RandomRangeFloat(Minimum=2, Maximum=3)
    )
    Scale RGB=&MakeVectorFromLinearColorRGB(Color=&ColorFromCurve(ColorCurve=DataInterface() {propertyValues {
        AlphaCurve {keys [<at 0 1 RCIM_Cubic>]}
        BlueCurve {keys [<at 0 3.5 RCIM_Cubic>,<at 0.2811 0.85>,<at 0.9319 0.36 RCIM_Cubic>]}
        GreenCurve {keys [<at 0 0.595 RCIM_Cubic>,<at 0.2811 0.595>,<at 0.9319 0.1908 RCIM_Cubic>]}
        RedCurve {keys [<at 0 2.0475 RCIM_Cubic>,<at 0.2811 0.697>,<at 0.9319 0.2134 RCIM_Cubic>]}
       }}))
    ScaleA=true
    ScaleRGB=true
   ) {enabled=false, module=ScaleColor001}
   &ScaleSpriteSize(Uniform Curve Sprite Scale=DataInterface() {propertyValues {Curve {keys [
        {interpMode=RCIM_Cubic, leaveTangent=11.1801, leaveTangentWeight=0.8986, tangentMode=RCTM_Break, time=0, value=0.9}
        {
         arriveTangent=1.5481
         arriveTangentWeight=0.2836
         interpMode=RCIM_Cubic
         leaveTangent=1.5481
         leaveTangentWeight=0.4242
         tangentMode=RCTM_User
         time=0.3
         value=2.3
        }
        {arriveTangent=1.0116, arriveTangentWeight=0.1994, interpMode=RCIM_Cubic, leaveTangent=1.0116, tangentMode=RCTM_User, time=1, value=4.4}
       ]}}})
   &ScaleColor(
    Scale RGB=&VectorFromFloat(Value=&RandomRangeFloat(Minimum=0.8, Maximum=1.2))
    ScaleA=false
    ScaleRGB=true
   ) {module=ScaleColor002}
  ]
  rendererClasses [&NiagaraSpriteRendererProperties, &NiagaraLightRendererProperties]
  renderers [
   {rendererClass=&NiagaraSpriteRendererProperties}
   {rendererClass=&NiagaraLightRendererProperties, rendererIndex=1}
  ]
 }
 SurtleSmoke {
  emitterUpdateScript [
   &EmitterState()
   &SpawnRate(
    SpawnRate=&FloatFromCurve(FloatCurve=DataInterface() {propertyValues {Curve {keys [<at 0 1>,<at 1 0>]}}}, CurveIndex=$Emitter.NormalizedLoopAge, Scale Curve=70)
    Use Spawn Probability=false
   ) {enabled=false}
   &SpawnBurst_Instantaneous(Spawn Count=10, Use Spawn Probability=false, UseLoopCountLimit=false)
  ]
  particleSpawnScript [
   &InitializeParticle(
    Color=&HueShiftLinearColor(
     Hue=0
     Saturation=1
     Value=&RandomRangeFloat(Minimum=0.1, Maximum=1)
     Linear Color=<rgba 0.0123 0.0014 0.026 1>
    )
    Lifetime Max=0.9
    Lifetime Min=0.6
    Lifetime Mode=::NewEnumerator1
    Mass Mode=::NewEnumerator3
    Sprite Rotation Mode=::NewEnumerator2
    Sprite Size Mode=::NewEnumerator3
    Uniform Sprite Size=&Multiply_Float(A=18, B=$User.FXScale)
    UsePositionOffset=false
    Write Material Random=true
   )
   /SetVariables(Particles.RandomFrame=&RandomRangeInt(Minimum=0, Maximum=63)) {isSetParameters=true}
   &AddVelocity(
    Cone Angle=90
    Cone Axis=<0 1 0>
    Module.UseConeFalloff=false
    Module.UseVelDistribution=false
    Override Seed=false
    Velocity Mode=::NewEnumerator2
    Velocity Speed=2453.7625
   )
   /SetVariables(Particles.Velocity=&Multiply_Vector(A=$Particles.Velocity, B=&VectorFromFloat(Value=0.3))) {isSetParameters=true}
  ]
  particleUpdateScript [
   &ParticleState()
   &Drag(Drag=5, Ignore Mass=false, Use Linear Drag=true, Use Rotational Drag=true)
   &SolveForcesAndVelocity(Clamp Velocity=false, Limit Acceleration=false)
   &ScaleSpriteSize(Scale Sprite Size Mode=::NewEnumerator0, Uniform Scale Factor=15.2886)
   &ScaleColor(
    Scale Alpha=&FloatFromCurve(
     FloatCurve=DataInterface() {propertyValues {Curve {keys [
         {arriveTangent=13.8559, interpMode=RCIM_Cubic, leaveTangent=13.8559, tangentMode=RCTM_User, time=0, value=0}
         {interpMode=RCIM_Cubic, tangentMode=RCTM_User, time=0.1, value=1}
         {interpMode=RCIM_Cubic, tangentMode=RCTM_User, time=0.5, value=1}
         {tangentMode=RCTM_User, time=1, value=0}
        ]}}}
     Scale Curve=0.8
    )
    ScaleA=true
    ScaleRGB=true
   )
   &DynamicMaterialParameters(
    Index 0 Param 1=&FloatFromCurve(
     FloatCurve=DataInterface() {propertyValues {Curve {keys [
         {arriveTangent=2.8718, interpMode=RCIM_Cubic, leaveTangent=2.8718, tangentMode=RCTM_User, time=0, value=0}
         {arriveTangent=-0.3942, interpMode=RCIM_Cubic, leaveTangent=-0.3942, tangentMode=RCTM_User, time=1, value=1}
        ]}}}
     Scale Curve=0.1
    )
    Module.Index0Param0Write=true
    Module.Index0Param1Write=true
    Module.Index0Param2Write=true
    Module.Index0Param3Write=true
   )
   &FN_SubUVAnimation(End Frame=&Add_Integer(A=$Particles.RandomFrame, B=33), Start Frame=$Particles.RandomFrame)
  ]
  rendererClasses [&NiagaraSpriteRendererProperties]
  renderers [{rendererClass=&NiagaraSpriteRendererProperties}]
 }
}
summary {
 systemName=NS_Proto_Ash_BugDeathFX_v3
 emitters [
  {emitterName=SurtleSmoke, rendererClasses [&NiagaraSpriteRendererProperties]}
  {
   bEnabled=false
   emitterName=Light
   rendererClasses [&NiagaraMeshRendererProperties, &NiagaraLightRendererProperties]
  }
  {emitterName=Directional_Splashes, rendererClasses [&NiagaraMeshRendererProperties]}
  {emitterName=Drops, rendererClasses [&NiagaraSpriteRendererProperties]}
  {emitterName=Splash_Center, rendererClasses [&NiagaraSpriteRendererProperties, &NiagaraLightRendererProperties]}
  {emitterName=NE_Splash_Directional, rendererClasses [&NiagaraSpriteRendererProperties]}
  {emitterName=MergedGlow, rendererClasses [&NiagaraSpriteRendererProperties]}
  {emitterName=LightStreaks, rendererClasses [&NiagaraSpriteRendererProperties]}
  {emitterName=Rainbow_Smoke, rendererClasses [&NiagaraSpriteRendererProperties]}
  {emitterName=Ashes, simTarget=GPUComputeSim, rendererClasses [&NiagaraSpriteRendererProperties]}
 ]
 userVariables [
  User.bIsFrontEnd {type=NiagaraBool}
  User.bIsFrontEndPreview {type=NiagaraBool}
  User.ElementSizeScale {type=NiagaraFloat, defaultValue=0, description=""}
  User.EnableGoo {type=NiagaraBool, defaultValue=true}
  User.FX_Scale_Global {type=NiagaraFloat, defaultValue=0, description=""}
  User.FXColor {type=NiagaraInt32, defaultValue=2}
  User.FXDecal {type=NiagaraBool}
  User.FXScale {type=NiagaraFloat, defaultValue=1, description=""}
  User.RandomArray {
   type=NiagaraDataInterfaceArrayFloat
   defaultValue {
    struct=NiagaraExt_VariableValue_DataInterface
    value {
     dataInterface=&self:NiagaraDataInterfaceArrayFloat_0
     dataInterfaceClass=&NiagaraDataInterfaceArrayFloat
    }
   }
  }
 ]
}
