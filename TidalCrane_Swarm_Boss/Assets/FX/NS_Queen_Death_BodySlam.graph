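# Niagara impact effect for the swarm boss Queen's death body slam, layering dirt, debris, smoke and purple goo along local X.
# Spawn groups split GPU dirt, mesh debris and dust to both sides; DamageAreaLength stretches smoke and goo boxes.
# Goo splashes, drops and growing, fading ground decals use EnableGoo as spawn probability.
# InitialDelay and Time Delay stagger several bursts and the dark purple smoke.
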NiagaraSystem /TidalCrane_Swarm_Boss/Assets/FX/NS_Queen_Death_BodySlam
dependencies {
 usedDataInterfaces [
  &NiagaraDataInterfaceArrayFloat3
  &NiagaraDataInterfaceColorCurve
  &NiagaraDataInterfaceCurve
  &NiagaraDataInterfaceArrayFloat
 ]
 usedDynamicInputs [&SimulationPosition]
 usedModules [
  &self:SystemSpawnScript.SystemScriptSource.SystemScriptGraph.NiagaraNodeAssignment_0.SetVariables_#1
  &SystemState
 ]
}
emitters {
 DecalRenderer001 {
  emitterUpdateScript [
   &EmitterState()
   &SpawnBurst_Instantaneous(
    Spawn Probability=&MakeCustomFloatFromBool(Bool=$User.EnableGoo, True Float=1, False Float=0)
    Use Spawn Probability=true
    UseLoopCountLimit=false
   )
   /SetVariables(
    Emitter.OpacityMultCrack:NiagaraFloat=&RandomRangeFloat(Minimum=1, Maximum=1.5)
    Emitter.RandRotCracks:NiagaraFloat=&RandomRangeFloat(Minimum=0, Maximum=1)
   ) {isSetParameters=true}
  ]
  particleSpawnScript [
   &InitializeParticle(Lifetime=10, Position Offset=<1750 0 0>, UsePositionOffset=true, Write Material Random=true)
   /SetVariables(Particles.DecalOrientationSpray:Quat4f=/Game/Effects/Niagara/Modules/OrientMestToVectorFromCustomRot(
     InitialRotation:Quat4f=<0 0 0 1>
     Look At Direction:Vector3f=<0 0 0>
     Reference Axis Vector:Vector3f=<0 0 0>
     RotationVector:Vector3f=&RandomRangeVector(Minimum=<0 0 0>, Maximum=<1 0 0>)
    )) {enabled=false, isSetParameters=true}
   /SetVariables(Particles.DecalPosSpray:NiagaraPosition=&AddVectorToPosition(Position=$Particles.Position, Vector=<0 0 0>)) {isSetParameters=true}
   /Niagara/Modules/Spawn/Decal/InitialDecalOrientation(
    Orientation Vector:Vector3f=<0 0 1>
    Rotation:Vector3f=&RandomRangeVector(Minimum=<0 0 -1>, Maximum=<0 0 1>)
    Use Rotation Vector=true
   )
  ]
  particleUpdateScript [
   &ParticleState()
   /SetVariables(Particles.DecalFadeSpray:NiagaraFloat=&Lerp_Float(
     A=0
     B=1
     Alpha=&FloatFromCurve(
      FloatCurve=DataInterface() {propertyValues {Curve {keys [
          {arriveTangent=6.7778, interpMode=RCIM_Cubic, leaveTangent=6.7778, tangentMode=RCTM_User, time=0, value=0.5}
          {arriveTangent=-0.3377, interpMode=RCIM_Cubic, leaveTangent=-0.3377, tangentMode=RCTM_User, time=0.1, value=1}
          {arriveTangent=-0.7311, interpMode=RCIM_Cubic, leaveTangent=-0.7311, tangentMode=RCTM_User, time=0.1, value=1}
          {arriveTangent=-0.6972, interpMode=RCIM_Cubic, leaveTangent=-0.6972, tangentMode=RCTM_User, time=0.7, value=0.67}
          {arriveTangent=-0.4726, interpMode=RCIM_Cubic, leaveTangent=-0.4726, tangentMode=RCTM_User, time=1, value=0}
         ]}}}
      Scale Curve=0.96
     )
    )) {isSetParameters=true}
   /SetVariables(Particles.DecalSizeSpray:Vector3f=&Lerp_Vector(
     A=<40 85 85>
     B=<40 750 750>
     Alpha=&FloatFromCurve(
      FloatCurve=DataInterface() {propertyValues {Curve {keys [
          {arriveTangent=35.2047, interpMode=RCIM_Cubic, leaveTangent=35.2047, tangentMode=RCTM_User, time=0, value=0.4}
          {arriveTangent=4.3794, interpMode=RCIM_Cubic, leaveTangent=4.3794, tangentMode=RCTM_User, time=0.02, value=0.8}
          {arriveTangent=0.7686, interpMode=RCIM_Cubic, leaveTangent=0.7686, tangentMode=RCTM_User, time=0.25, value=1}
         ]}}}
      Scale Curve=1.4
     )
    )) {isSetParameters=true}
   /SetVariables(Particles.DecalSizeThin:Vector3f=&Multiply_Vector(A=$Particles.DecalSizeSpray, B=&Multiply_VectorByFloat(Vector=<0 1 1>, Float=1.5))) {isSetParameters=true}
  ]
  rendererClasses [&NiagaraDecalRendererProperties]
  renderers [
   {rendererClass=&NiagaraDecalRendererProperties, rendererIndex=0}
   {rendererClass=&NiagaraDecalRendererProperties, rendererIndex=1}
  ]
 }
 Directional_Splashes {
  emitterUpdateScript [
   &EmitterState()
   &SpawnRate(
    Spawn Probability=&MakeCustomFloatFromBool(Bool=$User.EnableGoo, True Float=1, False Float=0)
    SpawnGroup=1
    SpawnRate=&Lerp_Float(A=100, B=20, Alpha=$Emitter.NormalizedLoopAge)
    Use Spawn Probability=true
   )
   &SpawnBurst_Instantaneous(
    Spawn Count=4
    Spawn Probability=&MakeCustomFloatFromBool(Bool=$User.EnableGoo, True Float=1, False Float=0)
    Spawn Time=0.05
    Use Spawn Probability=true
    UseLoopCountLimit=false
   )
  ]
  particleSpawnScript [
   &InitializeParticle(
    Color Maximum=<rgba 0.3642 0.23 0.46 1>
    Color Minimum=<rgba 0.1507 0.1276 0.22 1>
    Color Mode=::NewEnumerator2
    Lifetime Max=0.75
    Lifetime Min=0.3
    Lifetime Mode=::NewEnumerator1
    Mesh Scale=&MakeVector(
     X=&Multiply_Float(
      A=&MaskFloat(
       Mask Source=0
       Mask Parameter=?
       Mask=0
       Masked Float=1.275
       Passthrough Float=&RandomRangeFloat(Minimum=0.85, Maximum=1.25)
      )
      B=8
     )
     Y=1
     Z=&RandomRangeFloat(Minimum=1.75, Maximum=3.5)
    )
    Mesh Scale Mode=::NewEnumerator4
    Sprite Size=&Multiply_Vector2D_ByFloat(Float=0.2, Vector2D=&RandomRangeVector2D(Minimum=<150 200>, Maximum=<200 300>))
    Sprite Size Mode=::NewEnumerator4
    UsePositionOffset=false
    Write Material Random=true
   )
   &AddVelocity(
    Cone Angle=125
    Cone Axis=<0 0 1>
    Inner Cone Angle=90
    Module.UseConeFalloff=false
    Module.UseVelDistribution=false
    Override Seed=false
    Velocity Mode=::NewEnumerator2
    Velocity Speed=&MaskFloatBySpawnGroup(
     Masked Float=&Multiply_Float(
      A=&Lerp_Float(
       A=20
       B=150
       Alpha=&FloatFromCurve(
        FloatCurve=DataInterface() {propertyValues {Curve {keys [
            {arriveTangent=1.271, interpMode=RCIM_Cubic, leaveTangent=1.271, tangentMode=RCTM_User, time=0, value=0}
            {arriveTangent=3.4308, interpMode=RCIM_Cubic, leaveTangent=3.4308, tangentMode=RCTM_User, time=1, value=1}
           ]}}}
        CurveIndex=&RandomRangeFloat(Minimum=0, Maximum=1)
       )
      )
      B=1
     )
     Passthrough Float=30
     Spawn Group Mask=1
    )
   ) {module=AddVelocity001}
   &ShapeLocation(
    Box Midpoint=<0 0.5 0>
    Box Size=&MaskVector(
     Mask Source=0
     Mask Parameter=?
     Mask=1
     Masked Vector=&MakeVector(X=$User.DamageAreaLength, Y=150, Z=100)
     Passthrough Vector=<350 250 50>
    )
    MaskSpawnGroup=false
    Offset=&MaskVector(Mask Source=0, Mask Parameter=?, Mask=1, Masked Vector=<250 0 0>, Passthrough Vector=<1500 0 0>)
    Offset Coordinate Space=::Local
    Offset Mode=::NewEnumerator0
    Override Seed=false
    Shape Origin=$Engine.Emitter.SimulationPosition
    Shape Primitive=::NewEnumerator2
    Surface Only=false
    UseOwnerScale=true
   ) {module=ShapeLocation001}
   &AddVelocity(
    Module.Use Falloff Distance=false
    Origin Offset=<1450 0.2 -50>
    Override Seed=false
    Velocity Mode=::NewEnumerator1
    Velocity Speed=&RandomRangeFloat(Minimum=350, Maximum=550)
   )
   /SetVariables(Particles.Velocity=&Multiply_Vector(A=$Particles.Velocity, B=<0.35 1 0.35>)) {isSetParameters=true}
  ]
  particleUpdateScript [
   &ParticleState()
   &ScaleMeshSize(Scale Factor=&VectorFromFloat(Value=&RandomRangeFloat(Minimum=0.4, Maximum=0.5)))
   &Drag(Drag=&RandomRangeFloat(Minimum=2, Maximum=5), Use Linear Drag=true, Use Rotational Drag=false)
   &SolveForcesAndVelocity(Clamp Velocity=false, Limit Acceleration=false)
   &DynamicMaterialParameters(
    Index 0 Param 1=&UniformAOrBFloat(A=0, Evaluation Type=::NewEnumerator0, Recalculate Random Each Loop=true, Override Seed=false)
    Index 0 Param 2=&FloatFromCurve(FloatCurve=DataInterface() {propertyValues {Curve {keys [
         {arriveTangent=1.5496, interpMode=RCIM_Cubic, leaveTangent=1.5496, tangentMode=RCTM_User, time=0, value=0.2}
         {arriveTangent=0.3249, interpMode=RCIM_Cubic, leaveTangent=0.3249, tangentMode=RCTM_User, time=1, value=1}
        ]}}})
    Index 0 Param 3=&FloatFromCurve(
     FloatCurve=DataInterface() {propertyValues {Curve {keys [
         {arriveTangent=0.2226, interpMode=RCIM_Cubic, leaveTangent=0.2226, tangentMode=RCTM_User, time=0, value=0}
         {arriveTangent=1.6484, interpMode=RCIM_Cubic, leaveTangent=1.6484, tangentMode=RCTM_User, time=1, value=1}
        ]}}}
     Scale Curve=10
    )
    Module.Index0Param0Write=true
    Module.Index0Param1Write=true
    Module.Index0Param2Write=true
    Module.Index0Param3Write=true
   ) {module=DynamicMaterialParameters002}
   &ScaleColor(
    Scale Alpha=&FloatFromCurve(
     FloatCurve=DataInterface() {propertyValues {Curve {keys [
         {arriveTangent=-0.1364, interpMode=RCIM_Cubic, leaveTangent=-0.1364, tangentMode=RCTM_User, time=0.25, value=1}
         {arriveTangent=-3.5002, interpMode=RCIM_Cubic, leaveTangent=-3.5002, tangentMode=RCTM_User, time=1, value=0}
        ]}}}
     Scale Curve=&RandomRangeFloat(Minimum=0.8, Maximum=1.5)
    )
    Scale RGB=&MakeVectorFromLinearColorRGB(Color=&ColorFromCurve(ColorCurve=DataInterface() {propertyValues {
        AlphaCurve {keys [<at 0 1 RCIM_Cubic>]}
        BlueCurve {keys [<at 0 6 RCIM_Cubic>,<at 0.2324 1>,<at 1 0.5 RCIM_Cubic>]}
        GreenCurve {keys [<at 0 0.36 RCIM_Cubic>,<at 0.2324 0.74>,<at 1 0.35 RCIM_Cubic>]}
        RedCurve {keys [<at 0 3.18 RCIM_Cubic>,<at 0.2324 0.8267>,<at 1 0.35 RCIM_Cubic>]}
       }}))
    ScaleA=true
    ScaleRGB=true
   )
   /SetVariables(Particles.PositionWaves:NiagaraPosition=&AddVectorToPosition(
     Position=$Particles.Position
     Vector=&MakeVector(
      X=0
      Y=0
      Z=&Multiply_Float(
       A=&RandomRangeFloat(Minimum=-30, Maximum=-70)
       B=&FloatFromCurve(FloatCurve=DataInterface() {propertyValues {Curve {keys [
            {arriveTangent=0.0456, interpMode=RCIM_Cubic, leaveTangent=0.0456, tangentMode=RCTM_User, time=0, value=0}
            {arriveTangent=2.1323, interpMode=RCIM_Cubic, leaveTangent=2.1323, tangentMode=RCTM_User, time=1, value=1}
           ]}}})
      )
     )
    )) {isSetParameters=true}
   &ScaleColor(
    Scale RGB=&VectorFromFloat(Value=&RandomRangeFloat(Minimum=1.5, Maximum=4))
    ScaleA=false
    ScaleRGB=true
   ) {module=ScaleColor002}
  ]
  rendererClasses [&NiagaraMeshRendererProperties]
  renderers [
   {rendererClass=&NiagaraMeshRendererProperties, rendererIndex=0}
   {rendererClass=&NiagaraMeshRendererProperties, rendererIndex=1}
  ]
 }
 Dirt01 {
  simTarget=GPUComputeSim
  emitterUpdateScript [
   &EmitterState(Life Cycle Mode=::NewEnumerator1, Loop Behavior=::NewEnumerator1, Loop Duration=2, UseLoopDelay=false)
   &SpawnBurst_Instantaneous(Spawn Count=12, Spawn Time=$User.InitialDelay, Use Spawn Probability=false, UseLoopCountLimit=false)
   &SpawnBurst_Instantaneous(Spawn Count=12, Spawn Group=1, Spawn Time=0.175, Use Spawn Probability=false, UseLoopCountLimit=false) {module=SpawnBurst_Instantaneous001}
  ]
  particleSpawnScript [
   &InitializeParticle(
    Lifetime Max=1
    Lifetime Min=0.5
    Lifetime Mode=::NewEnumerator1
    Mass Mode=::NewEnumerator3
    Sprite Rotation Mode=::NewEnumerator2
    Sprite Size Mode=::NewEnumerator1
    Uniform Sprite Size Max=500
    Uniform Sprite Size Min=350
    UsePositionOffset=false
    Write Material Random=true
   )
   &CameraOffset(Camera Offset Amount=50)
   /SetVariables(Particles.GroupID:NiagaraInt32=$Emitter.SpawnGroup) {isSetParameters=true, module=SetVariables_#2}
   &ShapeLocation(
    Box Midpoint=<0 0.5 0.5>
    Box Size=&MakeVector(X=550, Y=100, Z=50)
    MaskSpawnGroup=false
    Offset=&SelectVectorFromArray(Direct Array Index=$Particles.GroupID, Vector Selection Array=DataInterface() {propertyValues {FloatData [<1250 -200 0>, <1250 200 0>]}})
    Offset Coordinate Space=::Local
    Offset Mode=::NewEnumerator0
    Override Seed=false
    Shape Origin=$Engine.Emitter.SimulationPosition
    Shape Primitive=::NewEnumerator2
    Surface Only=false
    UseOwnerScale=true
   )
   &AddVelocity(
    Cone Angle=50
    Cone Axis=&SelectVectorFromArray(Direct Array Index=$Particles.GroupID, Vector Selection Array=DataInterface() {propertyValues {FloatData [<1 -1 1>, <1 1 1>]}})
    Distribution Along Cone Axis=0
    Module.UseConeFalloff=false
    Module.UseVelDistribution=false
    Override Seed=false
    Speed Falloff From Cone Axis=0
    Velocity Mode=::NewEnumerator2
    Velocity Speed=&RandomRangeFloat(Minimum=800, Maximum=2250)
   ) {module=AddVelocity001}
  ]
  particleUpdateScript [
   &ParticleState()
   &Color(Color=&RandomRangeLinearColor(
     Minimum=&ColorFromCurve(ColorCurve=DataInterface() {propertyValues {
        AlphaCurve {keys [<at 0.4505 1 RCIM_Cubic>,<at 1 0 RCIM_Cubic>]}
        BlueCurve {keys [<at 0 0.2025 RCIM_Cubic>,<at 1 0.2025 RCIM_Cubic>]}
        GreenCurve {keys [<at 0 0.3859 RCIM_Cubic>,<at 1 0.3858 RCIM_Cubic>]}
        RedCurve {keys [<at 0 0.5972 RCIM_Cubic>,<at 1 0.5972 RCIM_Cubic>]}
       }})
     Maximum=&ColorFromCurve(ColorCurve=DataInterface() {propertyValues {
        AlphaCurve {keys [<at 0.4743 1 RCIM_Cubic>,<at 1 0 RCIM_Cubic>]}
        BlueCurve {keys [<at 0 0.0508 RCIM_Cubic>,<at 1 0.0508 RCIM_Cubic>]}
        GreenCurve {keys [<at 0 0.0974 RCIM_Cubic>,<at 1 0.0974 RCIM_Cubic>]}
        RedCurve {keys [<at 0 0.15 RCIM_Cubic>,<at 1 0.15 RCIM_Cubic>]}
       }})
    ))
   &GravityForce(Gravity=&RandomRangeVector(Minimum=<0 0 -1000>, Maximum=<1 1 -2000>))
   &Drag(
    Drag=&FloatFromCurve(
     FloatCurve=DataInterface() {propertyValues {Curve {
        defaultValue=1
        keys [
         {tangentMode=RCTM_User, time=0, value=0.2}
         {arriveTangent=-2.8845, interpMode=RCIM_Cubic, leaveTangent=-2.8845, tangentMode=RCTM_User, time=0.1, value=1}
         {tangentMode=RCTM_User, time=1, value=0.2}
        ]
       }}}
     Scale Curve=2
    )
    Use Linear Drag=true
    Use Rotational Drag=false
   )
   &SolveForcesAndVelocity(Clamp Velocity=false, Limit Acceleration=false)
  ]
  rendererClasses [&NiagaraSpriteRendererProperties]
  renderers [{rendererClass=&NiagaraSpriteRendererProperties}]
 }
 Drops {
  emitterUpdateScript [
   &EmitterState()
   &SpawnBurst_Instantaneous(
    Spawn Count=25
    Spawn Probability=&MakeCustomFloatFromBool(Bool=$User.EnableGoo, True Float=1, False Float=0)
    Spawn Time=$User.InitialDelay
    Use Spawn Probability=true
    UseLoopCountLimit=false
   )
   &SpawnBurst_Instantaneous(
    Spawn Count=100
    Spawn Group=1
    Spawn Probability=&MakeCustomFloatFromBool(Bool=$User.EnableGoo, True Float=1, False Float=0)
    Spawn Time=$User.Time Delay
    Use Spawn Probability=true
    UseLoopCountLimit=false
   ) {module=SpawnBurst_Instantaneous001}
  ]
  particleSpawnScript [
   &InitializeParticle(
    Color=&RandomRangeLinearColor(Minimum=<rgba 0.8 0.8 0.8 1>, Maximum=<rgba 1 1 1 1>)
    Lifetime=&UniformRangedFloat(Minimum=0.4, Maximum=1.4)
    Mass Mode=::NewEnumerator1
    Sprite Size=&Multiply_Vector2D_ByFloat(
     Float=&UniformAOrBFloat(
      A=&RandomRangeFloat(Minimum=3, Maximum=5)
      B=10
      Distribution Weight=0.1
      Evaluation Type=::NewEnumerator0
      Recalculate Random Each Loop=true
      Override Seed=false
     )
     Vector2D=&UniformRangedVector2D(Minimum=<2 3>, Maximum=<2 4>)
    )
    Sprite Size Mode=::NewEnumerator4
    UsePositionOffset=false
    Write Material Random=true
   ) {module=InitializeParticle001}
   &ShapeLocation(
    Box Midpoint=<0 0.5 0>
    Box Size=&MaskVector(
     Mask Source=0
     Mask Parameter=?
     Mask=1
     Masked Vector=&MakeVector(X=$User.DamageAreaLength, Y=200, Z=100)
     Passthrough Vector=<350 200 200>
    )
    MaskSpawnGroup=false
    Offset=&MaskVector(Mask Source=0, Mask Parameter=?, Mask=1, Masked Vector=<250 0 100>, Passthrough Vector=<1500 1 1>)
    Offset Coordinate Space=::Local
    Offset Mode=::NewEnumerator0
    Override Seed=false
    Shape Origin=$Engine.Emitter.SimulationPosition
    Shape Primitive=::NewEnumerator2
    Surface Only=false
    UseOwnerScale=true
   ) {module=ShapeLocation001}
   &AddVelocity(
    Cone Angle=&Lerp_Float(
     A=100
     B=300
     Alpha=&FloatFromCurve(
      FloatCurve=DataInterface() {propertyValues {Curve {keys [
          {arriveTangent=1.271, interpMode=RCIM_Cubic, leaveTangent=1.271, tangentMode=RCTM_User, time=0, value=0}
          {arriveTangent=3.4308, interpMode=RCIM_Cubic, leaveTangent=3.4308, tangentMode=RCTM_User, time=1, value=1}
         ]}}}
      CurveIndex=&RandomRangeFloat(Minimum=0, Maximum=1)
     )
    )
    Cone Axis=<0 0 1>
    Module.UseConeFalloff=false
    Module.UseVelDistribution=false
    Override Seed=false
    Velocity Mode=::NewEnumerator2
    Velocity Speed=&Multiply_Float(
     A=&Lerp_Float(
      A=1250
      B=4000
      Alpha=&FloatFromCurve(
       FloatCurve=DataInterface() {propertyValues {Curve {keys [
           {arriveTangent=1.271, interpMode=RCIM_Cubic, leaveTangent=1.271, tangentMode=RCTM_User, time=0, value=0}
           {arriveTangent=3.4308, interpMode=RCIM_Cubic, leaveTangent=3.4308, tangentMode=RCTM_User, time=1, value=1}
          ]}}}
       CurveIndex=&RandomRangeFloat(Minimum=0, Maximum=1)
      )
     )
     B=1
    )
   )
   /SetVariables(Particles.Velocity=&Multiply_Vector(A=$Particles.Velocity, B=<1 1 1>)) {isSetParameters=true}
   /Niagara/Modules/Update/SubUV/SubUVAnimation(SubUV Animation Mode=::NewEnumerator1) {module=SubUVAnimation001}
  ]
  particleUpdateScript [
   &ParticleState()
   &ScaleSpriteSize(Uniform Curve Sprite Scale=DataInterface() {propertyValues {Curve {keys [
        {arriveTangent=8.1474, interpMode=RCIM_Cubic, leaveTangent=8.1474, tangentMode=RCTM_User, time=0, value=0}
        {arriveTangentWeight=0.0928, interpMode=RCIM_Cubic, leaveTangentWeight=0.5258, tangentMode=RCTM_User, time=0.15, value=1}
        {arriveTangent=-2.0068, arriveTangentWeight=0.296, interpMode=RCIM_Cubic, leaveTangent=-2.0068, tangentMode=RCTM_User, time=1, value=0}
       ]}}}) {module=ScaleSpriteSize001}
   &Drag(Drag=&RandomRangeFloat(Minimum=2, Maximum=3), Use Linear Drag=true, Use Rotational Drag=true)
   &GravityForce(Gravity=&MakeVector(X=0, Y=0, Z=&RandomRangeFloat(Minimum=-980, Maximum=-1250))) {module=GravityForce001}
   &SolveForcesAndVelocity(Clamp Velocity=false, Limit Acceleration=false)
   &ScaleSpriteSizeBySpeed(Max Scale Factor=<0.66 8>, Min Scale Factor=<1 0.5>, Velocity Threshold=2500)
   &ScaleColor(
    Scale Alpha=&FloatFromCurve(
     FloatCurve=DataInterface() {propertyValues {Curve {keys [
         {arriveTangent=8.1018, interpMode=RCIM_Cubic, leaveTangent=8.1018, tangentMode=RCTM_User, time=0, value=0}
         {arriveTangentWeight=0.0573, interpMode=RCIM_Cubic, leaveTangentWeight=0.3728, tangentMode=RCTM_User, time=0.1, value=1}
         {arriveTangentWeight=0.2167, interpMode=RCIM_Cubic, leaveTangentWeight=0.0833, tangentMode=RCTM_User, time=0.75, value=1}
         <at 1 0 RCIM_Cubic>
        ]}}}
     Scale Curve=3
    )
    Scale RGB=&MakeVectorFromLinearColorRGB(Color=&ColorFromCurve(ColorCurve=DataInterface() {propertyValues {
        AlphaCurve {keys [<at 0 1 RCIM_Cubic>]}
        BlueCurve {keys [<at 0 3>, {arriveTangent=-2.34, interpMode=RCIM_Cubic, leaveTangent=-2.34, time=0.1504, value=0.9}, <at 1 0.66 RCIM_Cubic>]}
        GreenCurve {keys [<at 0 0.69>, {arriveTangent=-0.492, interpMode=RCIM_Cubic, leaveTangent=-0.492, time=0.1504, value=0.243}, <at 1 0.198 RCIM_Cubic>]}
        RedCurve {keys [<at 0 2.076>, {arriveTangent=-1.8395, interpMode=RCIM_Cubic, leaveTangent=-1.8395, time=0.1504, value=0.6372}, <at 1 0.2365 RCIM_Cubic>]}
       }}))
    ScaleA=true
    ScaleRGB=true
   )
   &DynamicMaterialParameters(
    Index 0 Param 1=&RandomRangeFloat(Minimum=0, Maximum=1)
    Index 0 Param 2=&FloatFromCurve(FloatCurve=DataInterface() {propertyValues {Curve {keys [<at 0.35 0.45>,<at 1 0.9>]}}})
    Index 0 Param 3=&FloatFromCurve(FloatCurve=DataInterface() {propertyValues {Curve {keys [{arriveTangent=2.6651, interpMode=RCIM_Cubic, leaveTangent=2.6651, tangentMode=RCTM_User, time=0, value=0}, <at 1 1 RCIM_Cubic>]}}})
    Index 0 Param 4=&FloatFromCurve(FloatCurve=DataInterface() {propertyValues {Curve {keys [<at 0 0 RCIM_Cubic>, {arriveTangent=2.332, interpMode=RCIM_Cubic, leaveTangent=2.332, tangentMode=RCTM_User, time=1, value=1}]}}})
    Module.Param0WriteEnabled=true
    Module.Param1WriteEnabled=true
    Module.Param2WriteEnabled=false
    Module.Param3WriteEnabled=true
   ) {module=DynamicMaterialParameters001}
  ]
  rendererClasses [&NiagaraSpriteRendererProperties]
  renderers [{rendererClass=&NiagaraSpriteRendererProperties}]
 }
 GOO_BigSplash {
  emitterUpdateScript [
   &EmitterState(Life Cycle Mode=::NewEnumerator1, Loop Behavior=::NewEnumerator1, Loop Duration=0.5, UseLoopDelay=true)
   &SpawnBurst_Instantaneous(
    Spawn Count=8
    Spawn Probability=&MakeCustomFloatFromBool(Bool=$User.EnableGoo, True Float=1, False Float=0)
    Use Spawn Probability=true
    UseLoopCountLimit=false
   )
   &SpawnRate(
    Spawn Probability=&MakeCustomFloatFromBool(Bool=$User.EnableGoo, True Float=1, False Float=0)
    SpawnGroup=1
    SpawnRate=&Lerp_Float(A=50, B=20, Alpha=$Emitter.NormalizedLoopAge)
    Use Spawn Probability=true
   )
   /SetVariables(Emitter.EmitterRandom:NiagaraFloat=&RandomRangeFloat(Minimum=0, Maximum=1)) {isSetParameters=true}
  ]
  particleSpawnScript [
   &InitializeParticle(
    Color=&RandomRangeLinearColor(Minimum=<rgba 0.8108 0.747 0.9 1>, Maximum=<rgba 0.8126 0.801 0.9 1>)
    Lifetime Max=1.15
    Lifetime Min=0.65
    Lifetime Mode=::NewEnumerator1
    Position Offset=<0 0 5>
    Sprite Rotation Mode=::NewEnumerator2
    Sprite Size Mode=::NewEnumerator1
    Uniform Sprite Size Max=1050
    Uniform Sprite Size Min=750
    UsePositionOffset=true
    Write Material Random=true
   )
   &AddVelocity(
    Cone Angle=120
    Cone Axis=<0 0 1>
    Inner Cone Angle=20
    Module.UseConeFalloff=false
    Module.UseVelDistribution=false
    Override Seed=false
    Velocity Mode=::NewEnumerator2
    Velocity Speed=&RandomRangeFloat(Minimum=350, Maximum=850)
   )
   &ShapeLocation(
    Box Midpoint=<0 0.5 0>
    Box Size=&MaskVector(
     Mask Source=0
     Mask Parameter=?
     Mask=1
     Masked Vector=&MakeVector(X=$User.DamageAreaLength, Y=350, Z=50)
     Passthrough Vector=<350 250 75>
    )
    MaskSpawnGroup=false
    Offset=&MaskVector(Mask Source=0, Mask Parameter=?, Mask=1, Masked Vector=<250 0 125>, Passthrough Vector=<1500 0 75>)
    Offset Coordinate Space=::Local
    Offset Mode=::NewEnumerator0
    Override Seed=false
    Shape Origin=$Engine.Emitter.SimulationPosition
    Shape Primitive=::NewEnumerator2
    Surface Only=false
    UseOwnerScale=true
   ) {module=ShapeLocation001}
   &CameraOffset(Camera Offset Amount=35)
  ]
  particleUpdateScript [
   &ParticleState()
   &DynamicMaterialParameters(
    Index 0 Param 1=$Emitter.EmitterRandom
    Index 0 Param 2=&FloatFromCurve(FloatCurve=DataInterface() {propertyValues {Curve {
        defaultValue=1
        keys [
         {arriveTangent=0.1559, interpMode=RCIM_Cubic, leaveTangent=0.1559, tangentMode=RCTM_User, time=0.2, value=0}
         {arriveTangent=1.8635, interpMode=RCIM_Cubic, leaveTangent=1.8635, tangentMode=RCTM_User, time=0.4, value=0.5}
         {arriveTangent=0.4386, interpMode=RCIM_Cubic, leaveTangent=0.4386, tangentMode=RCTM_User, time=1, value=1}
        ]
       }}})
    Index 0 Param 4=&Lerp_Float(
     A=&RandomRangeFloat(Minimum=4, Maximum=6)
     B=1
     Alpha=&FloatFromCurve(FloatCurve=DataInterface() {propertyValues {Curve {keys [
          {interpMode=RCIM_Cubic, tangentMode=RCTM_User, time=0, value=0}
          {arriveTangent=3.8492, interpMode=RCIM_Cubic, leaveTangent=3.8492, tangentMode=RCTM_User, time=0.26, value=0.3}
          {arriveTangent=0.9756, interpMode=RCIM_Cubic, leaveTangent=0.9756, tangentMode=RCTM_User, time=0.5, value=1}
         ]}}})
    )
    Module.Param0WriteEnabled=true
    Module.Param1WriteEnabled=true
    Module.Param2WriteEnabled=true
    Module.Param3WriteEnabled=true
   ) {module=DynamicMaterialParameters001}
   &ScaleSpriteSize(Uniform Curve Sprite Scale=DataInterface() {propertyValues {Curve {keys [
        {arriveTangent=1.4881, interpMode=RCIM_Cubic, leaveTangent=1.4881, tangentMode=RCTM_User, time=0, value=0}
        {tangentMode=RCTM_User, time=1.006, value=0.8824}
       ]}}})
   &GravityForce(Gravity=&RandomRangeVector(Minimum=<0 0 -550>, Maximum=<0 0 -1250>))
   &SolveForcesAndVelocity(Clamp Velocity=false, Limit Acceleration=false)
   &SpriteRotationRate(Rotation Rate=&RandomRangeFloat(Minimum=-20, Maximum=20))
   &ScaleColor(
    Scale Alpha=&FloatFromCurve(
     FloatCurve=DataInterface() {propertyValues {Curve {keys [
         {arriveTangent=17.6185, interpMode=RCIM_Cubic, leaveTangent=17.6185, tangentMode=RCTM_User, time=0, value=0}
         {arriveTangentWeight=0.0836, interpMode=RCIM_Cubic, leaveTangentWeight=0.3344, tangentMode=RCTM_User, time=0.07, value=1}
         {arriveTangent=-0.184, interpMode=RCIM_Cubic, leaveTangent=-0.184, leaveTangentWeight=0.6975, tangentMode=RCTM_User, time=0.34, value=1}
         {arriveTangentWeight=0.4545, interpMode=RCIM_Cubic, tangentMode=RCTM_User, time=1, value=0}
        ]}}}
     Scale Curve=&RandomRangeFloat(Minimum=2, Maximum=3)
    )
    Scale RGB=&MakeVectorFromLinearColorRGB(Color=&ColorFromCurve(ColorCurve=DataInterface() {propertyValues {
        AlphaCurve {keys [<at 0 1 RCIM_Cubic>]}
        BlueCurve {keys [<at 0 3.5 RCIM_Cubic>,<at 0.2811 0.85>,<at 0.9319 0.36 RCIM_Cubic>]}
        GreenCurve {keys [<at 0 0.595 RCIM_Cubic>,<at 0.2811 0.595>,<at 0.9319 0.1908 RCIM_Cubic>]}
        RedCurve {keys [<at 0 2.0475 RCIM_Cubic>,<at 0.2811 0.697>,<at 0.9319 0.2134 RCIM_Cubic>]}
       }}))
    ScaleA=true
    ScaleRGB=true
   )
   &ScaleColor(
    Scale RGB=&VectorFromFloat(Value=&RandomRangeFloat(Minimum=0.8, Maximum=1.2))
    ScaleA=false
    ScaleRGB=true
   ) {module=ScaleColor002}
  ]
  rendererClasses [&NiagaraSpriteRendererProperties]
  renderers [{rendererClass=&NiagaraSpriteRendererProperties}]
 }
 Mesh_Debris01 {
  emitterUpdateScript [
   &EmitterState(Life Cycle Mode=::NewEnumerator1, Loop Behavior=::NewEnumerator1, Loop Duration=1, UseLoopDelay=false)
   &SpawnBurst_Instantaneous(
    Spawn Count=&RandomRangeInt(Minimum=30, Maximum=50)
    Spawn Time=&Add_Float(A=$User.InitialDelay, B=0.025)
    Use Spawn Probability=false
    UseLoopCountLimit=false
   )
   &SpawnBurst_Instantaneous(
    Spawn Count=&RandomRangeInt(Minimum=30, Maximum=50)
    Spawn Group=1
    Spawn Time=&Add_Float(A=$User.InitialDelay, B=0.025)
    Use Spawn Probability=false
    UseLoopCountLimit=false
   ) {module=SpawnBurst_Instantaneous001}
  ]
  particleSpawnScript [
   &InitializeParticle(
    Color=<rgba 0.1788 0.0883 0.0319 1>
    Lifetime Max=1
    Lifetime Min=0.5
    Lifetime Mode=::NewEnumerator1
    Mass Mode=::NewEnumerator3
    Mesh Scale=&RandomRangeVector(Minimum=<0.1 0.2 0.1>, Maximum=<0.3 0.5 0.6>)
    Mesh Scale Mode=::NewEnumerator4
    Sprite Rotation Mode=::NewEnumerator2
    Sprite Size Mode=::NewEnumerator1
    Uniform Sprite Size Max=12
    Uniform Sprite Size Min=6
    UsePositionOffset=false
    Write Material Random=true
   )
   /SetVariables(Particles.GroupID:NiagaraInt32=$Emitter.SpawnGroup) {isSetParameters=true, module=SetVariables_#2}
   &ShapeLocation(
    Box Midpoint=<0 0.5 0.5>
    Box Size=&MakeVector(X=550, Y=100, Z=50)
    MaskSpawnGroup=false
    Offset=&SelectVectorFromArray(Direct Array Index=$Particles.GroupID, Vector Selection Array=DataInterface() {propertyValues {FloatData [<1250 -200 0>, <1250 200 0>]}})
    Offset Coordinate Space=::Local
    Offset Mode=::NewEnumerator0
    Override Seed=false
    Shape Origin=$Engine.Emitter.SimulationPosition
    Shape Primitive=::NewEnumerator2
    Surface Only=false
    UseOwnerScale=true
   )
   &AddVelocity(
    Cone Angle=50
    Cone Axis=&SelectVectorFromArray(Direct Array Index=$Particles.GroupID, Vector Selection Array=DataInterface() {propertyValues {FloatData [<1 -1 1.25>, <1 1 1.25>]}})
    Distribution Along Cone Axis=0
    Module.UseConeFalloff=false
    Module.UseVelDistribution=false
    Override Seed=false
    Speed Falloff From Cone Axis=0
    Velocity Mode=::NewEnumerator2
    Velocity Speed=&RandomRangeFloat(Minimum=1250, Maximum=2500)
   ) {module=AddVelocity001}
   &InitialMeshOrientation(Use Rotation Vector=false)
  ]
  particleUpdateScript [
   &ParticleState()
   &Color(Color=<rgba 0.9878 0.5348 0.3349 1>)
   &ScaleMeshSize(Scale Factor=&Multiply_VectorByFloat(
     Vector=<1 1 1>
     Float=&FloatFromCurve(FloatCurve=DataInterface() {propertyValues {Curve {keys [<at 0.5 1>,<at 1 0>]}}})
    ))
   &GravityForce(Gravity=&RandomRangeVector(Minimum=<0 0 -1500>, Maximum=<1 1 -2000>))
   &Drag(
    Drag=&FloatFromCurve(
     FloatCurve=DataInterface() {propertyValues {Curve {defaultValue=1, keys [{tangentMode=RCTM_User, time=0, value=0}, <at 0.3 0.5>, {tangentMode=RCTM_User, time=1, value=0}]}}}
     Scale Curve=5
    )
    Use Linear Drag=true
    Use Rotational Drag=false
   )
   &SolveForcesAndVelocity(Clamp Velocity=false, Limit Acceleration=false)
  ]
  rendererClasses [&NiagaraMeshRendererProperties]
  renderers [{rendererClass=&NiagaraMeshRendererProperties, rendererIndex=0}]
 }
 Smoke01 {
  emitterUpdateScript [
   &EmitterState(Life Cycle Mode=::NewEnumerator1, Loop Behavior=::NewEnumerator1, Loop Duration=0.25, UseLoopDelay=false)
   &SpawnRate(SpawnRate=50, Use Spawn Probability=false)
   &SpawnRate(SpawnGroup=1, SpawnRate=50, Use Spawn Probability=false) {module=SpawnRate001}
  ]
  particleSpawnScript [
   &InitializeParticle(
    Lifetime Max=1.25
    Lifetime Min=0.65
    Lifetime Mode=::NewEnumerator1
    Mass Mode=::NewEnumerator3
    Sprite Rotation Mode=::NewEnumerator2
    Sprite Size Mode=::NewEnumerator1
    Uniform Sprite Size Max=650
    Uniform Sprite Size Min=450
    UsePositionOffset=false
    Write Material Random=true
   )
   &CameraOffset(Camera Offset Amount=70)
   /SetVariables(Particles.GroupID:NiagaraInt32=$Emitter.SpawnGroup) {isSetParameters=true, module=SetVariables_#2}
   &ShapeLocation(
    Box Midpoint=<0 0.5 0>
    Box Size=&MakeVector(
     X=&Multiply_Float(
      A=$User.DamageAreaLength
      B=&FloatFromCurve(FloatCurve=DataInterface() {propertyValues {Curve {keys [<at 0 0>,<at 1 1>]}}}, CurveIndex=$Emitter.NormalizedLoopAge)
     )
     Y=100
     Z=50
    )
    MaskSpawnGroup=false
    Offset=&SelectVectorFromArray(Direct Array Index=$Particles.GroupID, Vector Selection Array=DataInterface() {propertyValues {FloatData [<500 -200 0>, <500 200 0>]}})
    Offset Coordinate Space=::Local
    Offset Mode=::NewEnumerator0
    Override Seed=false
    Shape Origin=$Engine.Emitter.SimulationPosition
    Shape Primitive=::NewEnumerator2
    Surface Only=false
    UseOwnerScale=true
   )
   &AddVelocity(
    Cone Angle=30
    Cone Axis=&SelectVectorFromArray(
     Direct Array Index=$Particles.GroupID
     Vector Selection Array=DataInterface() {propertyValues {FloatData [<0.5 -1 1.25>, <0.5 1 1.25>, <0 0 1>]}}
    )
    Distribution Along Cone Axis=0
    Module.UseConeFalloff=false
    Module.UseVelDistribution=false
    Override Seed=false
    Speed Falloff From Cone Axis=150
    Velocity Mode=::NewEnumerator2
    Velocity Speed=&RandomRangeFloat(Minimum=850, Maximum=2250)
   ) {module=AddVelocity001}
  ]
  particleUpdateScript [
   &ParticleState()
   /SetVariables(Particles.SubImageIndex=&FloatFromCurve(FloatCurve=DataInterface() {propertyValues {Curve {keys [
         {arriveTangent=29.8623, interpMode=RCIM_Cubic, leaveTangent=29.8623, tangentMode=RCTM_User, time=-0.0046, value=5.0556}
         <at 1.0011 22.1689 RCIM_Cubic>
        ]}}})) {isSetParameters=true}
   &DynamicMaterialParameters(
    Index 0 Param 1=&FloatFromCurve(FloatCurve=DataInterface() {propertyValues {Curve {keys [<at -0.0046 1.0113 RCIM_Cubic>,<at 1.0011 0.002>]}}})
    Index 0 Param 4=&FloatFromCurve(FloatCurve=DataInterface() {propertyValues {Curve {keys [<at 0 0 RCIM_Cubic>,<at 1 0.5>]}}})
    Module.Param0WriteEnabled=true
    Module.Param1WriteEnabled=false
    Module.Param2WriteEnabled=false
    Module.Param3WriteEnabled=true
   )
   &Color(Color=&RandomRangeLinearColor(
     Minimum=&ColorFromCurve(ColorCurve=DataInterface() {propertyValues {
        AlphaCurve {keys [<at 0.4505 0.3 RCIM_Cubic>,<at 1 0 RCIM_Cubic>]}
        BlueCurve {keys [<at 0 0.2025 RCIM_Cubic>,<at 1 0.2025 RCIM_Cubic>]}
        GreenCurve {keys [<at 0 0.3859 RCIM_Cubic>,<at 1 0.3858 RCIM_Cubic>]}
        RedCurve {keys [<at 0 0.5972 RCIM_Cubic>,<at 1 0.5972 RCIM_Cubic>]}
       }})
     Maximum=&ColorFromCurve(ColorCurve=DataInterface() {propertyValues {
        AlphaCurve {keys [<at 0.4743 0.15 RCIM_Cubic>,<at 1 0 RCIM_Cubic>]}
        BlueCurve {keys [<at 0 0.0508 RCIM_Cubic>,<at 1 0.0508 RCIM_Cubic>]}
        GreenCurve {keys [<at 0 0.0974 RCIM_Cubic>,<at 1 0.0974 RCIM_Cubic>]}
        RedCurve {keys [<at 0 0.15 RCIM_Cubic>,<at 1 0.15 RCIM_Cubic>]}
       }})
    ))
   &GravityForce(Gravity=&RandomRangeVector(Minimum=<0 0 -1000>, Maximum=<1 1 -2500>))
   &Drag(
    Drag=&FloatFromCurve(
     FloatCurve=DataInterface() {propertyValues {Curve {defaultValue=1, keys [{tangentMode=RCTM_User, time=0, value=0}, <at 0.3 0.5>, {tangentMode=RCTM_User, time=1, value=0}]}}}
     Scale Curve=&SelectFloatFromArray(Direct Array Index=$Particles.GroupID, Float Selection Array=DataInterface() {propertyValues {FloatData [4,4,20]}})
    )
    Use Linear Drag=true
    Use Rotational Drag=false
   )
   &SolveForcesAndVelocity(Clamp Velocity=false, Limit Acceleration=false)
  ]
  rendererClasses [&NiagaraSpriteRendererProperties]
  renderers [{rendererClass=&NiagaraSpriteRendererProperties}]
 }
 SurtleSmoke {
  emitterUpdateScript [
   &EmitterState(Life Cycle Mode=::NewEnumerator1, Loop Behavior=::NewEnumerator1, Loop Delay=$User.Time Delay, Loop Duration=1, UseLoopDelay=true)
   &SpawnRate(
    SpawnRate=&FloatFromCurve(FloatCurve=DataInterface() {propertyValues {Curve {keys [<at 0 1>,<at 1 0>]}}}, CurveIndex=$Emitter.NormalizedLoopAge, Scale Curve=25)
    Use Spawn Probability=false
   )
   &SpawnBurst_Instantaneous(Spawn Count=8, Use Spawn Probability=false, UseLoopCountLimit=false)
  ]
  particleSpawnScript [
   &InitializeParticle(
    Color=&HueShiftLinearColor(
     Hue=0
     Saturation=1
     Value=&RandomRangeFloat(Minimum=0.1, Maximum=1)
     Linear Color=<rgba 0.0123 0.0014 0.026 1>
    )
    Lifetime Max=0.9
    Lifetime Min=0.6
    Lifetime Mode=::NewEnumerator1
    Mass Mode=::NewEnumerator3
    Sprite Rotation Mode=::NewEnumerator2
    Sprite Size Mode=::NewEnumerator3
    Uniform Sprite Size=&Multiply_Float(A=55, B=$User.FXScale)
    UsePositionOffset=false
    Write Material Random=true
   )
   /SetVariables(Particles.RandomFrame=&RandomRangeInt(Minimum=0, Maximum=63)) {isSetParameters=true}
   &AddVelocity(Module.Use Falloff Distance=false, Override Seed=false, Velocity Mode=::NewEnumerator1, Velocity Speed=2500)
   &ShapeLocation(
    Box Midpoint=<0 0.5 0>
    Box Size=&MakeVector(X=$User.DamageAreaLength, Y=100, Z=150)
    MaskSpawnGroup=false
    Offset=<250 0 100>
    Offset Coordinate Space=::Local
    Offset Mode=::NewEnumerator0
    Override Seed=false
    Shape Origin=$Engine.Emitter.SimulationPosition
    Shape Primitive=::NewEnumerator2
    Surface Only=false
    UseOwnerScale=true
   ) {module=ShapeLocation001}
  ]
  particleUpdateScript [
   &ParticleState()
   &Drag(Drag=8, Ignore Mass=false, Use Linear Drag=true, Use Rotational Drag=true)
   &SolveForcesAndVelocity(Clamp Velocity=false, Limit Acceleration=false)
   &ScaleSpriteSize(Scale Sprite Size Mode=::NewEnumerator0, Uniform Scale Factor=15.2886)
   &ScaleColor(
    Scale Alpha=&FloatFromCurve(
     FloatCurve=DataInterface() {propertyValues {Curve {keys [
         {arriveTangent=13.8559, interpMode=RCIM_Cubic, leaveTangent=13.8559, tangentMode=RCTM_User, time=0, value=0}
         {interpMode=RCIM_Cubic, tangentMode=RCTM_User, time=0.1, value=1}
         {interpMode=RCIM_Cubic, tangentMode=RCTM_User, time=0.5, value=1}
         {tangentMode=RCTM_User, time=1, value=0}
        ]}}}
     Scale Curve=0.8
    )
    ScaleA=true
    ScaleRGB=true
   )
   &DynamicMaterialParameters(
    Index 0 Param 1=&FloatFromCurve(
     FloatCurve=DataInterface() {propertyValues {Curve {keys [
         {arriveTangent=2.8718, interpMode=RCIM_Cubic, leaveTangent=2.8718, tangentMode=RCTM_User, time=0, value=0}
         {arriveTangent=-0.3942, interpMode=RCIM_Cubic, leaveTangent=-0.3942, tangentMode=RCTM_User, time=1, value=1}
        ]}}}
     Scale Curve=0.1
    )
    Module.Index0Param0Write=true
    Module.Index0Param1Write=true
    Module.Index0Param2Write=true
    Module.Index0Param3Write=true
   )
   &FN_SubUVAnimation(End Frame=&Add_Integer(A=$Particles.RandomFrame, B=33), Start Frame=$Particles.RandomFrame)
  ]
  rendererClasses [&NiagaraSpriteRendererProperties]
  renderers [{rendererClass=&NiagaraSpriteRendererProperties}]
 }
}
summary {
 systemName=NS_Queen_Death_BodySlam
 emitters [
  {emitterName=Dirt01, simTarget=GPUComputeSim, rendererClasses [&NiagaraSpriteRendererProperties]}
  {emitterName=Mesh_Debris01, rendererClasses [&NiagaraMeshRendererProperties]}
  {emitterName=Smoke01, rendererClasses [&NiagaraSpriteRendererProperties]}
  {emitterName=SurtleSmoke, rendererClasses [&NiagaraSpriteRendererProperties]}
  {emitterName=GOO_BigSplash, rendererClasses [&NiagaraSpriteRendererProperties]}
  {emitterName=DecalRenderer001, rendererClasses [&NiagaraDecalRendererProperties]}
  {emitterName=Directional_Splashes, rendererClasses [&NiagaraMeshRendererProperties]}
  {emitterName=Drops, rendererClasses [&NiagaraSpriteRendererProperties]}
 ]
 userVariables [
  User.bIsFrontEnd {type=NiagaraBool}
  User.bIsFrontEndPreview {type=NiagaraBool}
  User.DamageAreaLength {type=NiagaraFloat, defaultValue=1750, description=""}
  User.DamageAreaWidth {type=NiagaraFloat, defaultValue=500, description=""}
  User.EnableGoo {type=NiagaraBool, defaultValue=true}
  User.FXColor {type=NiagaraInt32}
  User.FXScale {type=NiagaraFloat, defaultValue=1, description=""}
  User.InitialDelay {type=NiagaraFloat, defaultValue=0.1, description=""}
  User.Time Delay {type=NiagaraFloat, defaultValue=0.2, description=""}
 ]
}
