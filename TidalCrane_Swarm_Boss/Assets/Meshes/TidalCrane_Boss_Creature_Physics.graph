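# Physics asset for the TidalCrane boss creature, giving its skeleton simplified capsule bodies for collision and physics.
# One capsule each covers the tail, two spine segments, both jaws, and the shoulder, elbow and wrist of four arms.
# Constraints chain each arm to the spine, upper pair and jaws to spine_05, lower pair to spine_02.
# Arm and jaw joints swing and twist freely, unlike the two spine-tail links.
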PhysicsAsset /TidalCrane_Swarm_Boss/Assets/Meshes/TidalCrane_Boss_Creature_Physics
bodies {
 jaw [{shapeType=Capsule, length=90, radius=40, shapeName=jaw_capsule, center=<22.5196 23.9074 0.0103>, rotation=<rot 0.002 -80.0004 89.9809>}]
 lower_elbow_l [{
   shapeType=Capsule
   length=65
   radius=18.4469
   shapeName=lower_elbow_l_capsule
   center=<40.3898 0.5401 2.5035>
   rotation=<rot 0 92.7886 -84.1614>
  }]
 lower_elbow_r [{
   shapeType=Capsule
   length=65
   radius=18.4469
   shapeName=lower_elbow_r_capsule
   center=<-40.3898 -0.5401 -2.5035>
   rotation=<rot 0 92.7886 95.8386>
  }]
 lower_shoulder_l [{
   shapeType=Capsule
   length=60
   radius=20
   shapeName=lower_shoulder_l_capsule
   center=<38.3161 0.0001 0.0001>
   rotation=<rot -0.0002 -89.9999 -90.0001>
  }]
 lower_shoulder_r [{
   shapeType=Capsule
   length=60
   radius=20
   shapeName=lower_shoulder_r_capsule
   center=<-38.3161 -0.0001 -0.0001>
   rotation=<rot -0.0002 -89.9999 89.9999>
  }]
 lower_wrist_l [{
   shapeType=Capsule
   length=65
   radius=16.6284
   shapeName=lower_wrist_l_capsule
   center=<30.4049 3.9628 -26.0029>
   rotation=<rot -2.4966 -90 -33.4196>
  }]
 lower_wrist_r [{
   shapeType=Capsule
   length=65
   radius=16.6284
   shapeName=lower_wrist_r_capsule
   center=<-30.4049 -3.9628 26.0029>
   rotation=<rot -2.4966 -90 146.5804>
  }]
 reverse_jaw [{
   shapeType=Capsule
   length=82.1149
   radius=50
   shapeName=reverse_jaw_capsule
   center=<33.1444 -17.5981 4.0423>
   rotation=<rot 79.9115 93.8965 -117.785>
  }]
 spine_02 [{shapeType=Capsule, length=120, radius=75, center=<111.8222 -3.9561 0.0752>, rotation=<rot 0.2494 91.5045 -89.6438>}]
 spine_05 [{shapeType=Capsule, length=97.2647, radius=70.7272, center=<55.4264 -18.1597 -10.431>, rotation=<rot 15.1579 -90.0002 -90.8919>}]
 tail_02 [{shapeType=Capsule, length=100, radius=75, center=<82.5408 -5.1037 -0.0177>, rotation=<rot 0 82.2774 -89.9888>}]
 upper_elbow_l [{shapeType=Capsule, length=90, radius=25, shapeName=upper_elbow_l_capsule, center=<50.2037 0.8159 -0.9737>, rotation=<rot 0 92.838 -82.716>}]
 upper_elbow_r [{shapeType=Capsule, length=90, radius=25, shapeName=upper_elbow_r_capsule, center=<-50.2037 -0.8159 0.9737>, rotation=<rot 0 92.838 97.284>}]
 upper_shoulder_l [{
   shapeType=Capsule
   length=90
   radius=20
   shapeName=upper_shoulder_l_capsule
   center=<48.2196 0.0001 0.0001>
   rotation=<rot -0.0003 -89.9999 89.9999>
  }]
 upper_shoulder_r [{
   shapeType=Capsule
   length=90
   radius=20
   shapeName=upper_shoulder_r_capsule
   center=<-48.2196 -0.0001 -0.0001>
   rotation=<rot -0.0003 -89.9999 -90.0001>
  }]
 upper_wrist_l [{
   shapeType=Capsule
   length=90
   radius=20.2319
   shapeName=upper_wrist_l_capsule
   center=<37.2129 3.4638 -40.0544>
   rotation=<rot -2.7467 -90 -31.2469>
  }]
 upper_wrist_r [{
   shapeType=Capsule
   length=90
   radius=20.2319
   shapeName=upper_wrist_r_capsule
   center=<-37.2129 -3.4638 40.0544>
   rotation=<rot -2.7467 -90 148.7531>
  }]
}
constraints [
 {bone1Name=spine_02, bone2Name=tail_02, swing1LimitDegrees=45, swing2LimitDegrees=45, twistLimitDegrees=45}
 {bone1Name=spine_05, bone2Name=spine_02, swing1LimitDegrees=45, swing2LimitDegrees=45, twistLimitDegrees=45}
 {
  bone1Name=upper_elbow_r
  bone2Name=upper_shoulder_r
  swing1LimitDegrees=45
  swing1Motion=Free
  swing2LimitDegrees=45
  swing2Motion=Free
  twistLimitDegrees=45
  twistMotion=Free
 }
 {
  bone1Name=lower_shoulder_r
  bone2Name=spine_02
  swing1LimitDegrees=45
  swing1Motion=Free
  swing2LimitDegrees=45
  swing2Motion=Free
  twistLimitDegrees=45
  twistMotion=Free
 }
 {
  bone1Name=upper_wrist_l
  bone2Name=upper_elbow_l
  swing1LimitDegrees=45
  swing1Motion=Free
  swing2LimitDegrees=45
  swing2Motion=Free
  twistLimitDegrees=45
  twistMotion=Free
 }
 {
  bone1Name=upper_elbow_l
  bone2Name=upper_shoulder_l
  swing1LimitDegrees=45
  swing1Motion=Free
  swing2LimitDegrees=45
  swing2Motion=Free
  twistLimitDegrees=45
  twistMotion=Free
 }
 {
  bone1Name=upper_wrist_r
  bone2Name=upper_elbow_r
  swing1LimitDegrees=45
  swing1Motion=Free
  swing2LimitDegrees=45
  swing2Motion=Free
  twistLimitDegrees=45
  twistMotion=Free
 }
 {
  bone1Name=lower_wrist_l
  bone2Name=lower_elbow_l
  swing1LimitDegrees=45
  swing1Motion=Free
  swing2LimitDegrees=45
  swing2Motion=Free
  twistLimitDegrees=45
  twistMotion=Free
 }
 {
  bone1Name=lower_elbow_l
  bone2Name=lower_shoulder_l
  swing1LimitDegrees=45
  swing1Motion=Free
  swing2LimitDegrees=45
  swing2Motion=Free
  twistLimitDegrees=45
  twistMotion=Free
 }
 {
  bone1Name=lower_shoulder_l
  bone2Name=spine_02
  swing1LimitDegrees=45
  swing1Motion=Free
  swing2LimitDegrees=45
  swing2Motion=Free
  twistLimitDegrees=45
  twistMotion=Free
 }
 {
  bone1Name=lower_wrist_r
  bone2Name=lower_elbow_r
  swing1LimitDegrees=45
  swing1Motion=Free
  swing2LimitDegrees=45
  swing2Motion=Free
  twistLimitDegrees=45
  twistMotion=Free
 }
 {
  bone1Name=lower_elbow_r
  bone2Name=lower_shoulder_r
  swing1LimitDegrees=45
  swing1Motion=Free
  swing2LimitDegrees=45
  swing2Motion=Free
  twistLimitDegrees=45
  twistMotion=Free
 }
 {
  bone1Name=upper_shoulder_l
  bone2Name=spine_05
  swing1LimitDegrees=45
  swing1Motion=Free
  swing2LimitDegrees=45
  swing2Motion=Free
  twistLimitDegrees=45
  twistMotion=Free
 }
 {
  bone1Name=upper_shoulder_r
  bone2Name=spine_05
  swing1LimitDegrees=45
  swing1Motion=Free
  swing2LimitDegrees=45
  swing2Motion=Free
  twistLimitDegrees=45
  twistMotion=Free
 }
 {
  bone1Name=jaw
  bone2Name=spine_05
  swing1LimitDegrees=45
  swing1Motion=Free
  swing2LimitDegrees=45
  swing2Motion=Free
  twistLimitDegrees=45
  twistMotion=Free
 }
 {
  bone1Name=reverse_jaw
  bone2Name=spine_05
  swing1LimitDegrees=45
  swing1Motion=Free
  swing2LimitDegrees=45
  swing2Motion=Free
  twistLimitDegrees=45
  twistMotion=Free
 }
]
