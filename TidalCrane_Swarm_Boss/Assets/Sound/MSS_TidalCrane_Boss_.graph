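# One-shot mono MetaSound source for Tidal Crane boss audio, routed to the boss-fight VOX sound class.
# On play, it picks a random wave from the Wavs array, then rolls a pitch shift between Pitch.Min and Pitch.Max before playback.
# The Gain(dB) input multiplies the audio directly, so it acts as linear gain despite its name.
# Wave completion passes through a Trigger Delay to signal OnFinished.
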MetaSoundSource /TidalCrane_Swarm_Boss/Assets/Sound/MSS_TidalCrane_Boss_
inputs [
 "Gain(dB)" {type=Float}
 Pitch.Max {type=Float, access=Constructor}
 Pitch.Min {type=Float, access=Constructor}
 UE.Source.OnPlay {type=Trigger}
 Wavs {type=WaveAsset:Array, access=Constructor}
]
interfaces [UE.OutputFormat.Mono:1.0, UE.Source.OneShot:1.0, UE.Source:1.1]
outputs [
 "UE.OutputFormat.Mono.Audio:0" {type=Audio}
 UE.Source.OneShot.OnFinished {type=Trigger}
]
pages [Default {
  input_defaults {"Gain(dB)"=1.122, Pitch.Max=1, Pitch.Min=0, UE.Source.OnPlay=null, Wavs=null}
  nodes [
   ArrayRandomGet_0 = Array.Random Get:WaveAsset:Array(Next=^UE.Source.OnPlay, In Array=^Wavs) {also {defaults {Weights []}}}
   n1 = UE.RandomFloat(Max=^Pitch.Max, Min=^Pitch.Min, Next=@ArrayRandomGet_0.On Next, Reset=@ArrayRandomGet_0.On Reset)
   WavePlayer_0 = UE.Wave Player:Mono(Wave Asset=@ArrayRandomGet_0.Value, Play=@n1.On Next, Pitch Shift=@n1.Value) {also {version [1,1]}}
   Multiply_0 = UE.Multiply:Audio by Float(AdditionalOperands=^"Gain(dB)", PrimaryOperand=@WavePlayer_0.Out Mono)
   TriggerDelay_0 = UE.Trigger Delay(In=@WavePlayer_0.On Finished)
  ]
  links [{dst="@UE.OutputFormat.Mono.Audio:0", src=Multiply_0.Out}, {dst="@UE.Source.OneShot.OnFinished", src=TriggerDelay_0.Out}]
 }]
properties {
 assetImportData=&self:AssetImportData
 sampleRate=48000
 soundClassObject=/Game/Sounds/FortSoundClassesAndModes/Fort_BossFight_SoundClass/Fort_BossFight_VOX
}
