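# Mono one-shot MetaSound Source preset for the TidalCrane swarm boss death vocal, fed by three QueenBug death waves.
# The graph comes from a parent boss MetaSound; this file supplies inputs for waves, gain, pitch range, player-distance range, delay and reverb.
# Delay and reverb dry/wet ranges are zeroed here.
# Playback uses the bug boss roar attenuation and the boss-fight VOX sound class.
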MetaSoundSource /TidalCrane_Swarm_Boss/Assets/Sound/MSS_TidalCrane_Boss_Death
preset_parent=&/MSS_TidalCrane_Boss_
inputs [
 Delay.DryWet.Max {type=Float}
 Delay.DryWet.Min {type=Float}
 Delay.Feedback {type=Float, access=Constructor}
 Delay.Time {type=Time, access=Constructor}
 "Gain(dB)" {type=Float}
 MaxDistanceFromPlayer {type=Float, access=Constructor}
 MinDistanceFromPlayer {type=Float, access=Constructor}
 Pitch.Max {type=Float, access=Constructor}
 Pitch.Min {type=Float, access=Constructor}
 UE.Attenuation.Distance {type=Float}
 UE.Source.OnPlay {type=Trigger}
 Verb.Bandwidth {type=Float, access=Constructor}
 Verb.Dampening {type=Float, access=Constructor}
 Verb.Decay {type=Float, access=Constructor}
 "Verb.Delay\\(ms)" {type=Float, access=Constructor}
 Verb.Density {type=Float, access=Constructor}
 Verb.Diffusion {type=Float, access=Constructor}
 Verb.DryWet.Max {type=Float}
 Verb.DryWet.Min {type=Float}
 "Verb.Gain(dB)" {type=Float, access=Constructor}
 Wavs {type=WaveAsset:Array, access=Constructor}
]
interfaces [UE.Attenuation:1.0, UE.OutputFormat.Mono:1.0, UE.Source.OneShot:1.0, UE.Source:1.1]
outputs [
 "UE.OutputFormat.Mono.Audio:0" {type=Audio}
 UE.Source.OneShot.OnFinished {type=Trigger}
]
pages [Default {input_defaults {
   Delay.DryWet.Max=0
   Delay.DryWet.Min=0
   Delay.Feedback=0.45
   Delay.Time=0.38
   "Gain(dB)"=3
   MaxDistanceFromPlayer=25000
   MinDistanceFromPlayer=5000
   Pitch.Max=1
   Pitch.Min=0
   UE.Attenuation.Distance=0
   UE.Source.OnPlay=null
   Verb.Bandwidth=0.5
   Verb.Dampening=0.5
   Verb.Decay=0.368
   "Verb.Delay\\(ms)"=0.068
   Verb.Density=0.552
   Verb.Diffusion=0.544
   Verb.DryWet.Max=0
   Verb.DryWet.Min=0
   "Verb.Gain(dB)"=0.056
   Wavs [
    &/QueenBug_Death_03
    &/QueenBug_Death_02
    &/QueenBug_Death_01
   ]
  }}]
properties {
 assetImportData=&self:AssetImportData
 attenuationSettings=/TidalCrane_Swarm_Boss/Assets/Sound/Settings/Bug_Boss_Roar_VOX_Att
 maxDistance=26000
 sampleRate=48000
 soundClassObject=/Game/Sounds/FortSoundClassesAndModes/Fort_BossFight_SoundClass/Fort_BossFight_VOX
 volume=2
}
