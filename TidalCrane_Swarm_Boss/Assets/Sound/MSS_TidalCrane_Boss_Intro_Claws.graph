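# MetaSound Source preset that supplies the claw sound for the Tidal Crane swarm boss intro as a mono one-shot.
# It inherits its graph from a shared boss parent, feeding it three QueenBug claw wave variations, a gain boost and a pitch range.
# Playback uses the bug boss roar/earth-movement attenuation, routes to the boss-fight VOX sound class and overrides concurrency to cap simultaneous instances.
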MetaSoundSource /TidalCrane_Swarm_Boss/Assets/Sound/MSS_TidalCrane_Boss_Intro_Claws
preset_parent=&/MSS_TidalCrane_Boss_
inputs [
 "Gain(dB)" {type=Float}
 Pitch.Max {type=Float, access=Constructor}
 Pitch.Min {type=Float, access=Constructor}
 UE.Source.OnPlay {type=Trigger}
 Wavs {type=WaveAsset:Array, access=Constructor}
]
interfaces [UE.OutputFormat.Mono:1.0, UE.Source.OneShot:1.0, UE.Source:1.1]
outputs [
 "UE.OutputFormat.Mono.Audio:0" {type=Audio}
 UE.Source.OneShot.OnFinished {type=Trigger}
]
pages [Default {input_defaults {
   "Gain(dB)"=2
   Pitch.Max=1
   Pitch.Min=0
   UE.Source.OnPlay=null
   Wavs [
    &/QueenBug_Inro_Claws_01
    &/QueenBug_Inro_Claws_02
    &/QueenBug_Inro_Claws_03
   ]
  }}]
properties {
 assetImportData=&self:AssetImportData
 attenuationSettings=/TidalCrane_Swarm_Boss/Assets/Sound/Settings/Bug_Boss_Roar_EarthMovement_Att
 bOverrideConcurrency=true
 maxDistance=16500
 sampleRate=48000
 soundClassObject=/Game/Sounds/FortSoundClassesAndModes/Fort_BossFight_SoundClass/Fort_BossFight_VOX
 volume=1.5
 concurrencyOverrides {maxCount=2}
}
