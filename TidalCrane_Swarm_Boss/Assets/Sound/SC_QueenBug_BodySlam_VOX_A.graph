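# Sound cue for the Queen Bug boss's body slam vocalization in the Tidal Crane swarm boss content.
# A Random node picks one of four equally weighted voice variations and feeds it through a Modulator at the root.
# It plays on the boss fight VOX sound class with boss roar attenuation, a long audible range and slightly reduced volume.
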SoundCue /TidalCrane_Swarm_Boss/Assets/Sound/SC_QueenBug_BodySlam_VOX_A
root=Modulator_0
nodes [
 {id=Modulator_0, children [Random_0]}
 {id=Random_0, children [WavePlayer_1,WavePlayer_2,WavePlayer_3,WavePlayer_4], properties {weights [1,1,1,1]}}
 {id=WavePlayer_1, properties {soundWaveAssetPtr=&/QueenBug_BodySlam_VOX_A_01}}
 {id=WavePlayer_2, properties {soundWaveAssetPtr=&/QueenBug_BodySlam_VOX_A_02}}
 {id=WavePlayer_3, properties {soundWaveAssetPtr=&/QueenBug_BodySlam_VOX_A_03}}
 {id=WavePlayer_4, properties {soundWaveAssetPtr=&/QueenBug_BodySlam_VOX_A_04}}
]
properties {
 attenuationSettings=/TidalCrane_Swarm_Boss/Assets/Sound/Settings/Bug_Boss_Roar_VOX_Att
 maxDistance=26000
 soundClassObject=/Game/Sounds/FortSoundClassesAndModes/Fort_BossFight_SoundClass/Fort_BossFight_VOX
 volumeMultiplier=0.9
}
