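# SC_QueenBug_Death is the Sound Cue for the Queen Bug boss's death vocalization in the Tidal Crane swarm boss content.
# A Random node picks one of three equally weighted death waves and feeds a Modulator that boosts volume at unchanged pitch.
# The cue adds a further volume multiplier, routes to the boss-fight VOX sound class, and uses a death-roar attenuation asset for long-range falloff.
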SoundCue /TidalCrane_Swarm_Boss/Assets/Sound/SC_QueenBug_Death
root=Modulator_0
nodes [
 {id=Modulator_0, children [Random_0], properties {pitchMax=1, pitchMin=1, volumeMax=1.8, volumeMin=1.8}}
 {id=Random_0, children [WavePlayer_1,WavePlayer_2,WavePlayer_3], properties {weights [1,1,1]}}
 {id=WavePlayer_1, properties {soundWaveAssetPtr=&/QueenBug_Death_01}}
 {id=WavePlayer_2, properties {soundWaveAssetPtr=&/QueenBug_Death_02}}
 {id=WavePlayer_3, properties {soundWaveAssetPtr=&/QueenBug_Death_03}}
]
properties {
 attenuationSettings=/TidalCrane_Swarm_Boss/Assets/Sound/Settings/Bug_Boss_DeathRoar_VOX_Att
 maxDistance=16000
 soundClassObject=/Game/Sounds/FortSoundClassesAndModes/Fort_BossFight_SoundClass/Fort_BossFight_VOX
 volumeMultiplier=2
}
