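# Sound cue for the QueenBug boss intro claw sounds in the TidalCrane swarm boss content.
# A Random node picks one of three equally weighted claw wave variations and feeds it through a Modulator that boosts volume at fixed pitch.
# It plays on the boss-fight VOX sound class, using the shared bug boss roar and earth-movement attenuation for long-range audibility, with an added cue-level volume boost.
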SoundCue /TidalCrane_Swarm_Boss/Assets/Sound/SC_QueenBug_Intro_Claws
root=Modulator_0
nodes [
 {id=Modulator_0, children [Random_0], properties {pitchMax=1, pitchMin=1, volumeMax=2, volumeMin=2}}
 {id=Random_0, children [WavePlayer_0,WavePlayer_1,WavePlayer_2], properties {weights [1,1,1]}}
 {id=WavePlayer_0, properties {soundWaveAssetPtr=&/QueenBug_Inro_Claws_01}}
 {id=WavePlayer_1, properties {soundWaveAssetPtr=&/QueenBug_Inro_Claws_02}}
 {id=WavePlayer_2, properties {soundWaveAssetPtr=&/QueenBug_Inro_Claws_03}}
]
properties {
 attenuationSettings=/TidalCrane_Swarm_Boss/Assets/Sound/Settings/Bug_Boss_Roar_EarthMovement_Att
 maxDistance=16500
 soundClassObject=/Game/Sounds/FortSoundClassesAndModes/Fort_BossFight_SoundClass/Fort_BossFight_VOX
 volumeMultiplier=1.5
}
