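# Sound cue for the Queen Bug boss's intro vocalization, routed to the boss-fight VOX sound class with a roar attenuation and long audible range.
# A quality-level node picks between two equally weighted random pools: the first two levels use four takes, the last two use three, omitting take 02.
# A fixed-gain Modulator at the root doubles volume on top of the cue's own doubled multiplier.
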SoundCue /TidalCrane_Swarm_Boss/Assets/Sound/SC_QueenBug_Intro_VOX_A
root=Modulator_0
nodes [
 {id=Modulator_0, children [QualityLevel_0], properties {volumeMax=2, volumeMin=2}}
 {id=QualityLevel_0, children [Random_0,Random_0,Random_1,Random_1]}
 {id=Random_0, children [WavePlayer_0,WavePlayer_5,WavePlayer_6,WavePlayer_7], properties {weights [1,1,1,1]}}
 {id=Random_1, children [WavePlayer_0,WavePlayer_6,WavePlayer_7], properties {bShouldExcludeFromBranchCulling=true, weights [1,1,1]}}
 {id=WavePlayer_0, properties {soundWaveAssetPtr=&/QueenBug_Intro_VOX_A_01}}
 {id=WavePlayer_5, properties {soundWaveAssetPtr=&/QueenBug_Intro_VOX_A_02}}
 {id=WavePlayer_6, properties {soundWaveAssetPtr=&/QueenBug_Intro_VOX_A_03}}
 {id=WavePlayer_7, properties {soundWaveAssetPtr=&/QueenBug_Intro_VOX_A_04}}
]
properties {
 attenuationSettings=/TidalCrane_Swarm_Boss/Assets/Sound/Settings/Bug_Boss_Roar_VOX_Att
 maxDistance=26000
 soundClassObject=/Game/Sounds/FortSoundClassesAndModes/Fort_BossFight_SoundClass/Fort_BossFight_VOX
 volumeMultiplier=2
}
