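# Sound cue for the Queen Bug boss's earth-moving sound during repositioning, routed to a boss-fight VOX sound class with a long audible range.
# It mixes two layers at equal volume input, each randomly picking one of three equally weighted waves.
# The main layer is boosted in volume under its own attenuation.
# The distant layer uses separate distant recordings, heavily lowered in pitch, with a distant attenuation setting.
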SoundCue /TidalCrane_Swarm_Boss/Assets/Sound/SC_QueenBug_Reposition_EarthMove
root=Mixer_0
nodes [
 {
  id=Attenuation_0
  children [Modulator_0]
  properties {attenuationSettings=/TidalCrane_Swarm_Boss/Assets/Sound/Settings/Bug_Boss_Roar_EarthMovement_Att}
 }
 {
  id=Attenuation_1
  children [Modulator_1]
  properties {attenuationSettings=/TidalCrane_Swarm_Boss/Assets/Sound/Settings/Bug_Boss_Roar_EarthMovement_Distant_Att}
 }
 {id=Mixer_0, children [Attenuation_0, Attenuation_1], properties {inputVolume [1,1]}}
 {id=Modulator_0, children [Random_0], properties {volumeMax=1.5, volumeMin=1.5}}
 {id=Modulator_1, children [Random_4], properties {pitchMax=0.6, pitchMin=0.55, volumeMax=1, volumeMin=1}}
 {id=Random_0, children [WavePlayer_0,WavePlayer_1,WavePlayer_2], properties {weights [1,1,1]}}
 {id=Random_4, children [WavePlayer_3,WavePlayer_4,WavePlayer_5], properties {weights [1,1,1]}}
 {id=WavePlayer_0, properties {soundWaveAssetPtr=&/QueenBug_Reposition_EarthMove_01}}
 {id=WavePlayer_1, properties {soundWaveAssetPtr=&/QueenBug_Reposition_EarthMove_02}}
 {id=WavePlayer_2, properties {soundWaveAssetPtr=&/QueenBug_Reposition_EarthMove_03}}
 {id=WavePlayer_3, properties {soundWaveAssetPtr=&/QueenBug_Reposition_Enter_EarthMove_DIstant_01}}
 {id=WavePlayer_4, properties {soundWaveAssetPtr=&/QueenBug_Reposition_Enter_EarthMove_DIstant_02}}
 {id=WavePlayer_5, properties {soundWaveAssetPtr=&/QueenBug_Reposition_Enter_EarthMove_DIstant_03}}
]
properties {maxDistance=27000, soundClassObject=/Game/Sounds/FortSoundClassesAndModes/Fort_BossFight_SoundClass/Fort_BossFight_VOX, volumeMultiplier=2}
