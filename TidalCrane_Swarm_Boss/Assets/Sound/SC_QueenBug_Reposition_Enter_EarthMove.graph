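# Sound cue for the earth-movement sound of the Queen Bug boss's reposition enter, on the boss-fight VOX sound class.
# It mixes a near layer and a distant layer, each with its own attenuation and fixed-volume modulator over randomized waves.
# The near layer switches by quality level between four-wave (first two levels) and three-wave (last two) random pools, while the distant layer picks from three waves.
# A concurrency override limits simultaneous instances.
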SoundCue /TidalCrane_Swarm_Boss/Assets/Sound/SC_QueenBug_Reposition_Enter_EarthMove
root=Mixer_0
nodes [
 {
  id=Attenuation_0
  children [Modulator_0]
  properties {attenuationSettings=/TidalCrane_Swarm_Boss/Assets/Sound/Settings/Bug_Boss_Roar_EarthMovement_Att}
 }
 {
  id=Attenuation_1
  children [Modulator_1]
  properties {attenuationSettings=/TidalCrane_Swarm_Boss/Assets/Sound/Settings/Bug_Boss_Roar_EarthMovement_Distant_Att}
 }
 {id=Mixer_0, children [Attenuation_0, Attenuation_1], properties {inputVolume [1,1]}}
 {id=Modulator_0, children [QualityLevel_0], properties {volumeMax=2, volumeMin=2}}
 {id=Modulator_1, children [Random_4], properties {volumeMax=1, volumeMin=1}}
 {id=QualityLevel_0, children [Random_0,Random_0,Random_1,Random_1]}
 {id=Random_0, children [WavePlayer_5,WavePlayer_6,WavePlayer_7,WavePlayer_8], properties {weights [1,1,1,1]}}
 {id=Random_1, children [WavePlayer_5,WavePlayer_6,WavePlayer_8], properties {bShouldExcludeFromBranchCulling=true, weights [1,1,1]}}
 {id=Random_4, children [WavePlayer_0,WavePlayer_1,WavePlayer_2], properties {weights [1,1,1]}}
 {id=WavePlayer_0, properties {soundWaveAssetPtr=&/QueenBug_Reposition_Enter_EarthMove_DIstant_01}}
 {id=WavePlayer_1, properties {soundWaveAssetPtr=&/QueenBug_Reposition_Enter_EarthMove_DIstant_02}}
 {id=WavePlayer_2, properties {soundWaveAssetPtr=&/QueenBug_Reposition_Enter_EarthMove_DIstant_03}}
 {id=WavePlayer_5, properties {soundWaveAssetPtr=&/QueenBug_Reposition_Enter_EarthMove_01}}
 {id=WavePlayer_6, properties {soundWaveAssetPtr=&/QueenBug_Reposition_Enter_EarthMove_02}}
 {id=WavePlayer_7, properties {soundWaveAssetPtr=&/QueenBug_Reposition_Enter_EarthMove_03}}
 {id=WavePlayer_8, properties {soundWaveAssetPtr=&/QueenBug_Reposition_Enter_EarthMove_04}}
]
properties {
 bOverrideConcurrency=true
 maxDistance=27000
 soundClassObject=/Game/Sounds/FortSoundClassesAndModes/Fort_BossFight_SoundClass/Fort_BossFight_VOX
 volumeMultiplier=2
 concurrencyOverrides {maxCount=2}
}
