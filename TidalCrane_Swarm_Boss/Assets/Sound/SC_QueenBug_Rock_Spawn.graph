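# Sound cue for the boulder spawn sound of the Queen Bug boss in the TidalCrane swarm boss fight.
# A Random node picks one of three equally weighted boulder spawn waves, then a Modulator applies random pitch variation and a fixed volume boost.
# The cue also raises pitch and volume, uses the boss rock impact attenuation and boss-fight VOX sound class, and caps simultaneous instances through a concurrency override.
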SoundCue /TidalCrane_Swarm_Boss/Assets/Sound/SC_QueenBug_Rock_Spawn
root=Modulator_0
nodes [
 {id=Modulator_0, children [Random_0], properties {pitchMax=1.1, pitchMin=0.8, volumeMax=1.5, volumeMin=1.5}}
 {id=Random_0, children [WavePlayer_6,WavePlayer_7,WavePlayer_8], properties {weights [1,1,1]}}
 {id=WavePlayer_6, properties {soundWaveAssetPtr=&/QueenBug_BoulderSpawn_01}}
 {id=WavePlayer_7, properties {soundWaveAssetPtr=&/QueenBug_BoulderSpawn_02}}
 {id=WavePlayer_8, properties {soundWaveAssetPtr=&/QueenBug_BoulderSpawn_03}}
]
properties {
 attenuationSettings=/TidalCrane_Swarm_Boss/Assets/Sound/Settings/Bug_Boss_Rock_Impact_Att
 bOverrideConcurrency=true
 maxDistance=16500
 pitchMultiplier=1.5
 soundClassObject=/Game/Sounds/FortSoundClassesAndModes/Fort_BossFight_SoundClass/Fort_BossFight_VOX
 volumeMultiplier=2
 concurrencyOverrides {maxCount=3}
}
