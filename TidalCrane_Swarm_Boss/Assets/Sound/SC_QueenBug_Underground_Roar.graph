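# Sound Cue for the Queen Bug boss's underground roar, routed to the boss fight VOX sound class with dedicated roar/earth-movement attenuation.
# A Quality Level node picks an evenly weighted Random pool: all five roar waves for the first two levels, a three-wave subset for the last two.
# A root Modulator holds volume steady, leaving pitch at its default range.
# A concurrency override limits the cue to one playing instance.
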SoundCue /TidalCrane_Swarm_Boss/Assets/Sound/SC_QueenBug_Underground_Roar
root=Modulator_0
nodes [
 {id=Modulator_0, children [QualityLevel_0], properties {volumeMax=1, volumeMin=1}}
 {id=QualityLevel_0, children [Random_0,Random_0,Random_1,Random_1]}
 {id=Random_0, children [WavePlayer_0,WavePlayer_1,WavePlayer_2,WavePlayer_3,WavePlayer_4], properties {weights [1,1,1,1,1]}}
 {id=Random_1, children [WavePlayer_0,WavePlayer_2,WavePlayer_4], properties {bShouldExcludeFromBranchCulling=true, weights [1,1,1]}}
 {id=WavePlayer_0, properties {soundWaveAssetPtr=&/QueenBug_Underground_Roar_01}}
 {id=WavePlayer_1, properties {soundWaveAssetPtr=&/QueenBug_Underground_Roar_02}}
 {id=WavePlayer_2, properties {soundWaveAssetPtr=&/QueenBug_Underground_Roar_03}}
 {id=WavePlayer_3, properties {soundWaveAssetPtr=&/QueenBug_Underground_Roar_04}}
 {id=WavePlayer_4, properties {soundWaveAssetPtr=&/QueenBug_Underground_Roar_05}}
]
properties {
 attenuationSettings=/TidalCrane_Swarm_Boss/Assets/Sound/Settings/Bug_Boss_Roar_EarthMovement_Att
 bOverrideConcurrency=true
 maxDistance=16500
 soundClassObject=/Game/Sounds/FortSoundClassesAndModes/Fort_BossFight_SoundClass/Fort_BossFight_VOX
 volumeMultiplier=0.7
 concurrencyOverrides {maxCount=1}
}
