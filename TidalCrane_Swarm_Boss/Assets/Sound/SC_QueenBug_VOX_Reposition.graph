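# Sound cue for the Queen Bug swarm boss's roar vocalization when it repositions.
# A quality-level node sends the first two levels to a random pick of four equally weighted roars and the last two to a three-roar pick that omits the third.
# A root modulator and the cue's multiplier both boost volume.
# It uses boss-fight VOX sound class and shared boss roar attenuation.
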SoundCue /TidalCrane_Swarm_Boss/Assets/Sound/SC_QueenBug_VOX_Reposition
root=Modulator_0
nodes [
 {id=Modulator_0, children [QualityLevel_0], properties {volumeMax=2, volumeMin=2}}
 {id=QualityLevel_0, children [Random_0,Random_0,Random_1,Random_1]}
 {id=Random_0, children [WavePlayer_0,WavePlayer_1,WavePlayer_2,WavePlayer_3], properties {weights [1,1,1,1]}}
 {id=Random_1, children [WavePlayer_0,WavePlayer_1,WavePlayer_3], properties {bShouldExcludeFromBranchCulling=true, weights [1,1,1]}}
 {id=WavePlayer_0, properties {soundWaveAssetPtr=&/QueenBug_Reposition_Roar_01}}
 {id=WavePlayer_1, properties {soundWaveAssetPtr=&/QueenBug_Reposition_Roar_02}}
 {id=WavePlayer_2, properties {soundWaveAssetPtr=&/QueenBug_Reposition_Roar_03}}
 {id=WavePlayer_3, properties {soundWaveAssetPtr=&/QueenBug_Reposition_Roar_04}}
]
properties {
 attenuationSettings=/TidalCrane_Swarm_Boss/Assets/Sound/Settings/Bug_Boss_Roar_VOX_Att
 maxDistance=26000
 soundClassObject=/Game/Sounds/FortSoundClassesAndModes/Fort_BossFight_SoundClass/Fort_BossFight_VOX
 volumeMultiplier=2
}
